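# Animation Blueprint for the Unstable Liquid Gun weapon mesh, animating canister pistons, exhausts and magazine hinges through firing and reloading.
# Three state machines drive those parts via per-bone layered blends, and the ScaleMag curve scales the magazine bone to zero.
# The canister blends empty-to-full by ammo, then picks quarter-step reload animations after a delay.
# The event graph polls B_Weapon_LiquidGun, updating ammo values (assuming 20 rounds) only while not reloading.
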AnimBlueprint /UnstableLiquidGun/Items/UnstableLiquidGun/Animation/UnstableLiquidGun_AnimBP
target_skeleton=/UnstableLiquidGun/Items/UnstableLiquidGun/Meshes/Unstable_Liquid_Gun_Skeleton
graphs [
 AnimGraph {nodes [
   n1 = Sequence Player 'UnstableLiquidGun_BasePose_Weapon'() {sequence=&/UnstableLiquidGun_BasePose_Weapon}
   AnimGraphNode_StateMachine_0 = InnerCanister() {is=title AnimGraphNode_StateMachine}
   AnimGraphNode_StateMachine_2 = Exhausts() {is=title AnimGraphNode_StateMachine}
   n2 = Layered blend per bone(
    BasePose=Layered blend per bone(BasePose=Slot 'DefaultSlot'(Source=@n1), BlendPoses_0=@AnimGraphNode_StateMachine_0) {layerSetup [{branchFilters [{boneName=piston_01}, {boneName=piston_02}]}]}
    BlendPoses_0=@AnimGraphNode_StateMachine_2
   ) {layerSetup [{branchFilters [{boneName=back_exhaust_l}, {boneName=back_exhaust_r}, {boneName=front_exhaust_l}, {boneName=front_exhaust_r}]}]}
   AnimGraphNode_StateMachine_1 = Eject() {is=title AnimGraphNode_StateMachine}
   n3 = Layered blend per bone(BasePose=@n2, BlendPoses_0=@AnimGraphNode_StateMachine_1) {layerSetup [{branchFilters [{boneName=magazine_hinge_01}, {boneName=magazine_hinge_02}, {boneName=magazine_hinge_03}, {boneName=magazine_hinge_04}]}]}
   n4 = "Transform (Modify) Bone - Bone: magazine"!(ComponentPose=Local To Component(LocalPose=@n3), Alpha=GetCurveValue(CurveName=ScaleMag), Scale="0,0,0") -> (Pose) {scaleMode=BMM_Replace, scaleSpace=BCS_BoneSpace, alphaBoolBlend {blendOption=Linear}, boneToModify {boneName=magazine}, scale=<0 0 0>}
   Output Pose(Result=Component To Local(ComponentPose=@n4)) {layerGroup=None}
  ]}
 InnerCanister {
  entry=CurrentPose
  path=AnimGraph.AnimGraphNode_StateMachine_0.InnerCanister
  ids=id
  nodes [
   CanisterRefill() {is=title AnimStateAliasNode, also {aliases [ReloadDelay]}}
   AnimStateNode_0 = DelayHold() {is=title AnimStateNode}
   AnimStateNode_1 = ChooseAmount() {is=title AnimStateNode}
   AnimStateNode_2 = CurrentPose() {is=title AnimStateNode}
   AnimStateNode_3 = Reload Quarter() {is=title AnimStateNode}
   AnimStateNode_4 = Reload Full() {is=title AnimStateNode}
   AnimStateNode_5 = Reload 3Quarter() {is=title AnimStateNode}
   AnimStateNode_6 = Reload Half() {is=title AnimStateNode}
   AnimStateNode_7 = ReloadDelay() {is=title AnimStateNode}
   AnimStateTransitionNode_10 = AnimStateTransitionNode() {crossfadeDuration=0}
   AnimStateTransitionNode_11 = AnimStateTransitionNode() {crossfadeDuration=0}
   AnimStateTransitionNode_12 = AnimStateTransitionNode()
   AnimStateTransitionNode_13 = AnimStateTransitionNode() {bAutomaticRuleBasedOnSequencePlayerInState=true}
   AnimStateTransitionNode_14 = AnimStateTransitionNode() {crossfadeDuration=0.1}
   AnimStateTransitionNode_15 = AnimStateTransitionNode() {crossfadeDuration=0.1}
   AnimStateTransitionNode_16 = AnimStateTransitionNode() {crossfadeDuration=0.1}
   AnimStateTransitionNode_17 = AnimStateTransitionNode() {crossfadeDuration=0.1}
   AnimStateTransitionNode_2 = AnimStateTransitionNode()
   AnimStateTransitionNode_8 = AnimStateTransitionNode() {crossfadeDuration=0}
   AnimStateTransitionNode_9 = AnimStateTransitionNode() {crossfadeDuration=0}
  ]
  links [
   {
    dst=ChooseAmount
    rule=AnimGraph.AnimGraphNode_StateMachine_0.InnerCanister.AnimStateTransitionNode_13.Transition
    src=CanisterRefill
    via=AnimStateTransitionNode_13
   }
   {
    dst=Reload 3Quarter
    rule=AnimGraph.AnimGraphNode_StateMachine_0.InnerCanister.AnimStateTransitionNode_9.Transition
    src=ChooseAmount
    via=AnimStateTransitionNode_9
   }
   {
    dst=Reload Full
    rule=AnimGraph.AnimGraphNode_StateMachine_0.InnerCanister.AnimStateTransitionNode_8.AnimationTransitionGraph_1
    src=ChooseAmount
    via=AnimStateTransitionNode_8
   }
   {
    dst=Reload Half
    rule=AnimGraph.AnimGraphNode_StateMachine_0.InnerCanister.AnimStateTransitionNode_10.Transition
    src=ChooseAmount
    via=AnimStateTransitionNode_10
   }
   {
    dst=Reload Quarter
    rule=AnimGraph.AnimGraphNode_StateMachine_0.InnerCanister.AnimStateTransitionNode_11.Transition
    src=ChooseAmount
    via=AnimStateTransitionNode_11
   }
   {
    dst=DelayHold
    rule=AnimGraph.AnimGraphNode_StateMachine_0.InnerCanister.AnimStateTransitionNode_2.Transition
    src=CurrentPose
    via=AnimStateTransitionNode_2
   }
   {
    dst=ReloadDelay
    rule=AnimGraph.AnimGraphNode_StateMachine_0.InnerCanister.AnimStateTransitionNode_12.Transition
    src=DelayHold
    via=AnimStateTransitionNode_12
   }
   {
    dst=CurrentPose
    rule=AnimGraph.AnimGraphNode_StateMachine_0.InnerCanister.AnimStateTransitionNode_14.Transition
    src=Reload 3Quarter
    via=AnimStateTransitionNode_14
   }
   {
    dst=CurrentPose
    rule=AnimGraph.AnimGraphNode_StateMachine_0.InnerCanister.AnimStateTransitionNode_17.Transition
    src=Reload Full
    via=AnimStateTransitionNode_17
   }
   {
    dst=CurrentPose
    rule=AnimGraph.AnimGraphNode_StateMachine_0.InnerCanister.AnimStateTransitionNode_15.Transition
    src=Reload Half
    via=AnimStateTransitionNode_15
   }
   {
    dst=CurrentPose
    rule=AnimGraph.AnimGraphNode_StateMachine_0.InnerCanister.AnimStateTransitionNode_16.Transition
    src=Reload Quarter
    via=AnimStateTransitionNode_16
   }
  ]
 }
 DelayHold {
  path=AnimGraph.AnimGraphNode_StateMachine_0.InnerCanister.AnimStateNode_0.DelayHold
  nodes [
   n1 = Get ReloadAmount() -> (ReloadAmount:double)
   Output Animation Pose(Result=Two Way Blend(
     A=Sequence Player 'UnstableLiquidGun_EmptyCanister_Weapon'() {sequence=&/UnstableLiquidGun_EmptyCanister_Weapon}
     B=Sequence Player 'UnstableLiquidGun_FullCanister_Weapon'() {sequence=&/UnstableLiquidGun_FullCanister_Weapon}
     Alpha=float - float(A=1, B=@n1)
    ) {blendNode {alphaBoolBlend {blendOption=Linear}}})
  ]
 }
 ChooseAmount {
  path=AnimGraph.AnimGraphNode_StateMachine_0.InnerCanister.AnimStateNode_1.ChooseAmount
  nodes [
   n1 = Get ReloadAmount()
   Output Animation Pose(Result=Two Way Blend(
     A=Sequence Player 'UnstableLiquidGun_EmptyCanister_Weapon'() {sequence=&UnstableLiquidGun_EmptyCanister_Weapon}
     B=Sequence Player 'UnstableLiquidGun_FullCanister_Weapon'() {sequence=&UnstableLiquidGun_FullCanister_Weapon}
     Alpha=float - float(A=1, B=@n1)
    ) {blendNode {alphaBoolBlend {blendOption=Linear}}})
  ]
 }
 CurrentPose {
  path=AnimGraph.AnimGraphNode_StateMachine_0.InnerCanister.AnimStateNode_2.CurrentPose
  nodes [
   Output Animation Pose(Result=Two Way Blend(
     A=Sequence Player 'UnstableLiquidGun_EmptyCanister_Weapon'() {sequence=&UnstableLiquidGun_EmptyCanister_Weapon}
     B=Sequence Player 'UnstableLiquidGun_FullCanister_Weapon'() {sequence=&UnstableLiquidGun_FullCanister_Weapon}
     Alpha=Get CanisterValue() -> (CanisterValue:double)
    ) {blendNode {alphaBoolBlend {blendOption=Linear}}})
   Get AmmoPercentage() -> (AmmoPercentage:double)
  ]
 }
 Reload Quarter {
  path=AnimGraph.AnimGraphNode_StateMachine_0.InnerCanister.AnimStateNode_3.Reload Quarter
  nodes [Output Animation Pose(Result=Sequence Player 'UnstableLiquidGun_Reload_Quarter_Weapon'() {bLoopAnimation=false, sequence=/UnstableLiquidGun/Items/UnstableLiquidGun/Animation/Reload/UnstableLiquidGun_Reload_Quarter_Weapon})]
 }
 Reload Full {
  path=AnimGraph.AnimGraphNode_StateMachine_0.InnerCanister.AnimStateNode_4.Reload Full
  nodes [Output Animation Pose(Result=Sequence Player 'UnstableLiquidGun_Reload_All_Weapon'() {bLoopAnimation=false, sequence=/UnstableLiquidGun/Items/UnstableLiquidGun/Animation/Reload/UnstableLiquidGun_Reload_All_Weapon})]
 }
 Reload 3Quarter {
  path=AnimGraph.AnimGraphNode_StateMachine_0.InnerCanister.AnimStateNode_5.Reload 3Quarter
  nodes [
   n1 = Sequence Player 'UnstableLiquidGun_Reload_3Quarters_Weapon'() {bLoopAnimation=false, sequence=/UnstableLiquidGun/Items/UnstableLiquidGun/Animation/Reload/UnstableLiquidGun_Reload_3Quarters_Weapon}
   Output Animation Pose(Result=@n1)
  ]
 }
 Reload Half {
  path=AnimGraph.AnimGraphNode_StateMachine_0.InnerCanister.AnimStateNode_6.Reload Half
  nodes [Output Animation Pose(Result=Sequence Player 'UnstableLiquidGun_Reload_Half_Weapon'() {bLoopAnimation=false, sequence=/UnstableLiquidGun/Items/UnstableLiquidGun/Animation/Reload/UnstableLiquidGun_Reload_Half_Weapon})]
 }
 ReloadDelay {
  path=AnimGraph.AnimGraphNode_StateMachine_0.InnerCanister.AnimStateNode_7.ReloadDelay
  nodes [
   n1 = Get ReloadAmount()
   Output Animation Pose(Result=Two Way Blend(
     A=Sequence Player 'UnstableLiquidGun_EmptyCanister_Weapon'() {sequence=&UnstableLiquidGun_EmptyCanister_Weapon}
     B=Sequence Player 'UnstableLiquidGun_FullCanister_Weapon'() {sequence=&UnstableLiquidGun_FullCanister_Weapon}
     Alpha=float - float(A=1, B=@n1)
    ) {blendNode {alphaBoolBlend {blendOption=Linear}}})
  ]
 }
 Transition {
  path=AnimGraph.AnimGraphNode_StateMachine_0.InnerCanister.AnimStateTransitionNode_10.Transition
  nodes [Result(bCanEnterTransition="Equal (Float)"(A=Get ReloadAmount(), B=0.5))]
 }
 Transition {
  path=AnimGraph.AnimGraphNode_StateMachine_0.InnerCanister.AnimStateTransitionNode_11.Transition
  nodes [Result(bCanEnterTransition="Equal (Float)"(A=Get ReloadAmount(), B=0.25))]
 }
 Transition {
  path=AnimGraph.AnimGraphNode_StateMachine_0.InnerCanister.AnimStateTransitionNode_12.Transition
  nodes [Result(bCanEnterTransition=Get ReloadDelay() -> (ReloadDelay:bool))]
 }
 Transition {path=AnimGraph.AnimGraphNode_StateMachine_0.InnerCanister.AnimStateTransitionNode_13.Transition, nodes [Result()]}
 Transition {
  path=AnimGraph.AnimGraphNode_StateMachine_0.InnerCanister.AnimStateTransitionNode_14.Transition
  nodes [Result(bCanEnterTransition=NOT Boolean(A=Get Reloading() -> (Reloading:bool)))]
 }
 Transition {
  path=AnimGraph.AnimGraphNode_StateMachine_0.InnerCanister.AnimStateTransitionNode_15.Transition
  nodes [Result(bCanEnterTransition=NOT Boolean(A=Get Reloading()))]
 }
 Transition {
  path=AnimGraph.AnimGraphNode_StateMachine_0.InnerCanister.AnimStateTransitionNode_16.Transition
  nodes [Result(bCanEnterTransition=NOT Boolean(A=Get Reloading()))]
 }
 Transition {
  path=AnimGraph.AnimGraphNode_StateMachine_0.InnerCanister.AnimStateTransitionNode_17.Transition
  nodes [Result(bCanEnterTransition=NOT Boolean(A=Get Reloading()))]
 }
 Transition {
  path=AnimGraph.AnimGraphNode_StateMachine_0.InnerCanister.AnimStateTransitionNode_2.Transition
  nodes [Result(bCanEnterTransition=Get Reloading())]
 }
 AnimationTransitionGraph_1 {
  path=AnimGraph.AnimGraphNode_StateMachine_0.InnerCanister.AnimStateTransitionNode_8.AnimationTransitionGraph_1
  nodes [Result(bCanEnterTransition="Equal (Float)"(A=Get ReloadAmount(), B=1))]
 }
 Transition {
  path=AnimGraph.AnimGraphNode_StateMachine_0.InnerCanister.AnimStateTransitionNode_9.Transition
  nodes [Result(bCanEnterTransition="Equal (Float)"(A=Get ReloadAmount(), B=0.75))]
 }
 Eject {
  entry=Idle
  path=AnimGraph.AnimGraphNode_StateMachine_1.Eject
  ids=id
  nodes [
   AnimStateNode_1 = Idle()
   AnimStateNode_2 = Eject() {is=title AnimStateNode}
   AnimStateNode_3 = Reloading() {is=title AnimStateNode}
   AnimStateTransitionNode_1 = AnimStateTransitionNode()
   AnimStateTransitionNode_2 = AnimStateTransitionNode() {crossfadeDuration=0.1}
   AnimStateTransitionNode_3 = AnimStateTransitionNode() {bAutomaticRuleBasedOnSequencePlayerInState=true, crossfadeDuration=0.25}
  ]
  links [
   {dst=Idle, rule=AnimGraph.AnimGraphNode_StateMachine_1.Eject.AnimStateTransitionNode_1.Transition, src=Eject, via=AnimStateTransitionNode_1}
   {
    dst=Reloading
    rule=AnimGraph.AnimGraphNode_StateMachine_1.Eject.AnimStateTransitionNode_2.Transition
    src=Idle
    via=AnimStateTransitionNode_2
   }
   {
    dst=Eject
    rule=AnimGraph.AnimGraphNode_StateMachine_1.Eject.AnimStateTransitionNode_3.Transition
    src=Reloading
    via=AnimStateTransitionNode_3
   }
  ]
 }
 Idle {
  path=AnimGraph.AnimGraphNode_StateMachine_1.Eject.AnimStateNode_1.Idle
  nodes [Output Animation Pose(Result=Sequence Player 'UnstableLiquid_EjectedPose_Weapon'() {sequence=&/UnstableLiquid_EjectedPose_Weapon})]
 }
 Eject {
  path=AnimGraph.AnimGraphNode_StateMachine_1.Eject.AnimStateNode_2.Eject
  nodes [Output Animation Pose(Result=Sequence Player 'UnstableLiquid_Eject_Weapon'() {bLoopAnimation=false, sequence=&/UnstableLiquid_Eject_Weapon})]
 }
 Reloading {
  path=AnimGraph.AnimGraphNode_StateMachine_1.Eject.AnimStateNode_3.Reloading
  nodes [Output Animation Pose(Result=Sequence Player 'UnstableLiquidGun_DelayForEject_Weapon'() {sequence=&/UnstableLiquidGun_DelayForEject_Weapon})]
 }
 Transition {
  path=AnimGraph.AnimGraphNode_StateMachine_1.Eject.AnimStateTransitionNode_1.Transition
  nodes [Result(bCanEnterTransition=NOT Boolean(A=Get Reloading()))]
 }
 Transition {path=AnimGraph.AnimGraphNode_StateMachine_1.Eject.AnimStateTransitionNode_2.Transition, nodes [Result(bCanEnterTransition=Get Reloading())]}
 Transition {path=AnimGraph.AnimGraphNode_StateMachine_1.Eject.AnimStateTransitionNode_3.Transition, nodes [Result()]}
 Exhausts {
  entry=BasePose
  path=AnimGraph.AnimGraphNode_StateMachine_2.Exhausts
  ids=id
  nodes [
   IsReloading() {is=title AnimStateAliasNode, also {aliases [BasePose,FireEnd,FireLoop]}}
   AnimStateNode_0 = BasePose() {
    is=title AnimStateNode
    stateEntered {guid=#1}
    stateFullyBlended {guid=#2}
    stateLeft {guid=#3}
   }
   AnimStateNode_1 = Reloading()
   AnimStateNode_5 = FireStart() {is=title AnimStateNode}
   AnimStateNode_6 = FireEnd() {
    is=title AnimStateNode
    stateEntered {guid=#4}
    stateFullyBlended {guid=#5}
    stateLeft {guid=#6}
   }
   AnimStateNode_7 = FireLoop() {is=title AnimStateNode}
   AnimStateTransitionNode_0 = AnimStateTransitionNode() {crossfadeDuration=0.1}
   AnimStateTransitionNode_1 = AnimStateTransitionNode() {bAutomaticRuleBasedOnSequencePlayerInState=true, crossfadeDuration=0.25}
   AnimStateTransitionNode_2 = AnimStateTransitionNode() {crossfadeDuration=0}
   AnimStateTransitionNode_24 = AnimStateTransitionNode() {bAutomaticRuleBasedOnSequencePlayerInState=true, crossfadeDuration=0.001}
   AnimStateTransitionNode_3 = AnimStateTransitionNode() {bAutomaticRuleBasedOnSequencePlayerInState=true, crossfadeDuration=0}
   AnimStateTransitionNode_5 = AnimStateTransitionNode() {crossfadeDuration=0}
  ]
  links [
   {
    dst=FireStart
    rule=AnimGraph.AnimGraphNode_StateMachine_2.Exhausts.AnimStateTransitionNode_2.Transition
    src=BasePose
    via=AnimStateTransitionNode_2
   }
   {
    dst=BasePose
    rule=AnimGraph.AnimGraphNode_StateMachine_2.Exhausts.AnimStateTransitionNode_3.Transition
    src=FireEnd
    via=AnimStateTransitionNode_3
   }
   {
    dst=FireEnd
    rule=AnimGraph.AnimGraphNode_StateMachine_2.Exhausts.AnimStateTransitionNode_5.Transition
    src=FireLoop
    via=AnimStateTransitionNode_5
   }
   {
    dst=FireLoop
    rule=AnimGraph.AnimGraphNode_StateMachine_2.Exhausts.AnimStateTransitionNode_24.Transition
    src=FireStart
    via=AnimStateTransitionNode_24
   }
   {
    dst=Reloading
    rule=AnimGraph.AnimGraphNode_StateMachine_2.Exhausts.AnimStateTransitionNode_0.Transition
    src=IsReloading
    via=AnimStateTransitionNode_0
   }
   {
    dst=BasePose
    rule=AnimGraph.AnimGraphNode_StateMachine_2.Exhausts.AnimStateTransitionNode_1.Transition
    src=Reloading
    via=AnimStateTransitionNode_1
   }
  ]
 }
 BasePose {
  path=AnimGraph.AnimGraphNode_StateMachine_2.Exhausts.AnimStateNode_0.BasePose
  nodes [Output Animation Pose(Result=Sequence Player 'UnstableLiquidGun_BasePose_Weapon'() {sequence=&UnstableLiquidGun_BasePose_Weapon})]
 }
 Reloading {
  path=AnimGraph.AnimGraphNode_StateMachine_2.Exhausts.AnimStateNode_1.Reloading
  nodes [Output Animation Pose(Result=Sequence Player 'UnstableLiquidWeapon_Reload'() {sequence=/UnstableLiquidGun/Items/UnstableLiquidGun/Animation/Reload/UnstableLiquidWeapon_Reload})]
 }
 FireStart {
  path=AnimGraph.AnimGraphNode_StateMachine_2.Exhausts.AnimStateNode_5.FireStart
  nodes [Output Animation Pose(Result=Sequence Player 'UnstableLiquid_Exhaust_FireStart_Weapon'() {bLoopAnimation=false, sequence=/UnstableLiquidGun/Items/UnstableLiquidGun/Animation/Exhaust/UnstableLiquid_Exhaust_FireStart_Weapon})]
 }
 FireEnd {
  path=AnimGraph.AnimGraphNode_StateMachine_2.Exhausts.AnimStateNode_6.FireEnd
  nodes [Output Animation Pose(Result=Sequence Player 'UnstableLiquid_Exhaust_FireEnd_Weapon'() {bLoopAnimation=false, sequence=/UnstableLiquidGun/Items/UnstableLiquidGun/Animation/Exhaust/UnstableLiquid_Exhaust_FireEnd_Weapon})]
 }
 FireLoop {
  path=AnimGraph.AnimGraphNode_StateMachine_2.Exhausts.AnimStateNode_7.FireLoop
  nodes [Output Animation Pose(Result=Sequence Player 'UnstableLiquid_Exhaust_FireLoop_Weapon'() {sequence=/UnstableLiquidGun/Items/UnstableLiquidGun/Animation/Exhaust/UnstableLiquid_Exhaust_FireLoop_Weapon})]
 }
 Transition {
  path=AnimGraph.AnimGraphNode_StateMachine_2.Exhausts.AnimStateTransitionNode_0.Transition
  nodes [Result(bCanEnterTransition=Get Reloading())]
 }
 Transition {path=AnimGraph.AnimGraphNode_StateMachine_2.Exhausts.AnimStateTransitionNode_1.Transition, nodes [Result()]}
 Transition {
  path=AnimGraph.AnimGraphNode_StateMachine_2.Exhausts.AnimStateTransitionNode_2.Transition
  nodes [Result(bCanEnterTransition=Get WeaponFired() -> (WeaponFired:bool))]
 }
 Transition {path=AnimGraph.AnimGraphNode_StateMachine_2.Exhausts.AnimStateTransitionNode_24.Transition, nodes [Result()]}
 Transition {path=AnimGraph.AnimGraphNode_StateMachine_2.Exhausts.AnimStateTransitionNode_3.Transition, nodes [Result()]}
 Transition {
  path=AnimGraph.AnimGraphNode_StateMachine_2.Exhausts.AnimStateTransitionNode_5.Transition
  nodes [Result(bCanEnterTransition=NOT Boolean(A=Get WeaponFired()))]
 }
 EventGraph {nodes [
   Event BlueprintUpdateAnimation()
   n1 = GetOwningActor()
   > Is Valid(InputObject=@n1)
   > n2 = Cast To B_Weapon_LiquidGun!(execute, Object=@n1) -> (then, CastFailed, AsB Weapon Liquid Gun:B Weapon Liquid Gun Object Reference) {is=title K2Node_DynamicCast}
   > Set Reloading(Reloading:bool=IsReloading!(self:Fort Weapon Object Reference=@n2) -> (ReturnValue:bool) {is=title K2Node_CallFunction}) -> (Output_Get:bool)
   > Set WeaponFired(WeaponFired:bool=AND Boolean(
     A=Get Firing!(self:B Weapon Liquid Gun Object Reference=@n2) -> (Firing:bool)
     B=Get bFiringMontagesActive!(self:B Weapon Liquid Gun Object Reference=@n2) -> (bFiringMontagesActive:bool)
    )) -> (Output_Get:bool)
   > n3 = Branch(Condition=Get Reloading())
   > Delay(Duration=0.75)
   > Set ReloadDelay(ReloadDelay:bool=true) -> (Output_Get:bool)
   Set ReloadDelay(execute=@n3.else, ReloadDelay=false)
   > n4 = Set RemainingAmmo(RemainingAmmo:int=GetLocalMagazineAmmoCount!(self:Fort Weapon Object Reference=@n2) -> (ReturnValue:int=0)) -> (Output_Get:int=0)
   > n5 = Set AmmoPercentage(AmmoPercentage:double=float / float(A:int=@n4, B:int=20)) -> (Output_Get:double)
   > Set CanisterValue(CanisterValue:double=@n5) -> (Output_Get:double)
   > Set ReloadAmount(ReloadAmount:double=float - float(A=1, B=Get AmmoPercentage())) -> (Output_Get:double)
  ]}
]
variables [
 AmmoPercentage {type=number}
 AmmoPercentageCache {type=number}
 CanisterValue {type=number}
 ReloadAmount {type=number}
 ReloadDelay {type=boolean}
 ReloadDone {type=boolean}
 Reloading {type=boolean}
 RemainingAmmo {type=integer}
 WeaponFired {type=boolean}
]
