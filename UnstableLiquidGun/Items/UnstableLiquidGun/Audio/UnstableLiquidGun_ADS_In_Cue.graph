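# Sound cue for the Unstable Liquid Gun's aim-down-sights-in foley, layering an aim movement sound with a tank slosh.
# A mixer plays one of four random ADS-in waves immediately, alongside a quieter, briefly delayed random pick from six low-tank slosh waves.
# Each layer passes through a modulator.
# It uses weapon foley class and reload attenuation, with concurrency settings limiting overlapping plays.
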SoundCue /UnstableLiquidGun/Items/UnstableLiquidGun/Audio/UnstableLiquidGun_ADS_In_Cue
root=Mixer_1
nodes [
 {id=Delay_0, children [Modulator_1], properties {delayMax=0.3, delayMin=0.3}}
 {id=Mixer_1, children [Modulator_0,Delay_0], properties {inputVolume [2,0.4]}}
 {id=Modulator_0, children [Random_0]}
 {id=Modulator_1, children [Random_2]}
 {id=Random_0, children [WavePlayer_0,WavePlayer_2,WavePlayer_3,WavePlayer_4], properties {weights [1,1,1,1]}}
 {
  id=Random_2
  children [WavePlayer_10, WavePlayer_11, WavePlayer_12, WavePlayer_13, WavePlayer_14, WavePlayer_15]
  properties {weights [1,1,1,1,1,1]}
 }
 {id=WavePlayer_0, properties {soundWaveAssetPtr=&/WPN_UnstableLiquidGun_ADS_In_01}}
 {
  id=WavePlayer_10
  properties {soundWaveAssetPtr=&/WPN_UnstableLiquidGun_Slosh_TankLow_Single_01}
 }
 {
  id=WavePlayer_11
  properties {soundWaveAssetPtr=&/WPN_UnstableLiquidGun_Slosh_TankLow_Single_02}
 }
 {
  id=WavePlayer_12
  properties {soundWaveAssetPtr=&/WPN_UnstableLiquidGun_Slosh_TankLow_Single_03}
 }
 {
  id=WavePlayer_13
  properties {soundWaveAssetPtr=&/WPN_UnstableLiquidGun_Slosh_TankLow_Single_04}
 }
 {
  id=WavePlayer_14
  properties {soundWaveAssetPtr=&/WPN_UnstableLiquidGun_Slosh_TankLow_Single_05}
 }
 {
  id=WavePlayer_15
  properties {soundWaveAssetPtr=&/WPN_UnstableLiquidGun_Slosh_TankLow_Single_06}
 }
 {id=WavePlayer_2, properties {soundWaveAssetPtr=&/WPN_UnstableLiquidGun_ADS_In_02}}
 {id=WavePlayer_3, properties {soundWaveAssetPtr=&/WPN_UnstableLiquidGun_ADS_In_03}}
 {id=WavePlayer_4, properties {soundWaveAssetPtr=&/WPN_UnstableLiquidGun_ADS_In_04}}
]
properties {
 attenuationSettings=&Fort_Weapon_Reload_Att
 maxDistance=2250
 soundClassObject=&Fort_WeaponFoley
 volumeMultiplier=0.351
 concurrencyOverrides {maxCount=1, resolutionRule=PreventNew}
 concurrencySet [&PickupMaximum_Oldest]
}
