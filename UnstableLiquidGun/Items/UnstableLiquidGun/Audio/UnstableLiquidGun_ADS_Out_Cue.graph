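# Sound cue for the Unstable Liquid Gun's foley when leaving aim-down-sights.
# A Random node picks one of four equally weighted ADS-out waves and feeds it through a Modulator at default settings for slight pitch and volume variation.
# It plays at reduced volume in the weapon foley sound class, borrowing the weapon reload attenuation and the PickupMaximum_Oldest concurrency set.
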SoundCue /UnstableLiquidGun/Items/UnstableLiquidGun/Audio/UnstableLiquidGun_ADS_Out_Cue
root=Modulator_0
nodes [
 {id=Modulator_0, children [Random_0]}
 {id=Random_0, children [WavePlayer_0,WavePlayer_2,WavePlayer_3,WavePlayer_4], properties {weights [1,1,1,1]}}
 {id=WavePlayer_0, properties {soundWaveAssetPtr=&/WPN_UnstableLiquidGun_ADS_Out_04}}
 {id=WavePlayer_2, properties {soundWaveAssetPtr=&/WPN_UnstableLiquidGun_ADS_Out_01}}
 {id=WavePlayer_3, properties {soundWaveAssetPtr=&/WPN_UnstableLiquidGun_ADS_Out_02}}
 {id=WavePlayer_4, properties {soundWaveAssetPtr=&/WPN_UnstableLiquidGun_ADS_Out_03}}
]
properties {
 attenuationSettings=&Fort_Weapon_Reload_Att
 maxDistance=2250
 soundClassObject=&Fort_WeaponFoley
 volumeMultiplier=0.501
 concurrencyOverrides {maxCount=1, resolutionRule=PreventNew}
 concurrencySet [&PickupMaximum_Oldest]
}
