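# First-person fire sound cue for the Unstable Liquid Gun, mixing a start one-shot with a sustained firing loop.
# The start layer randomly picks one of four waves with slight pitch variation.
# The half-volume loop layer picks one of two looping waves, shaped by an envelope that swells past unity, then eases back.
# It uses first-person weapon attenuation and owner-limited, stop-oldest concurrency overrides alongside shared gunfire concurrency sets.
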SoundCue /UnstableLiquidGun/Items/UnstableLiquidGun/Audio/UnstableLiquidGun_Fire_1P_Cue
root=Mixer_0
nodes [
 {
  id=Enveloper_0
  children [Random_1]
  properties {
   pitchCurve {editorCurveData {keys [{time=0}]}}
   volumeCurve {editorCurveData {keys [<at 0 0>,<at 0.382 0.2859>,<at 0.6569 0.8468>,<at 0.9 1.5>,<at 1.35 1>,<at 1.5822 0.779>]}}
  }
 }
 {id=Mixer_0, children [Modulator_1,Modulator_0], properties {inputVolume [1,0.5]}}
 {id=Modulator_0, children [Enveloper_0], properties {pitchMax=1, pitchMin=1}}
 {id=Modulator_1, children [Random_0], properties {pitchMax=0.95, pitchMin=1.05}}
 {id=Random_0, children [WavePlayer_1,WavePlayer_2,WavePlayer_4,WavePlayer_5], properties {weights [1,1,1,1]}}
 {id=Random_1, children [WavePlayer_6,WavePlayer_8], properties {weights [1,1]}}
 {id=WavePlayer_1, properties {soundWaveAssetPtr=&/WPN_UnstableLiquidGun_Fire_Start_Close_01}}
 {id=WavePlayer_2, properties {soundWaveAssetPtr=&/WPN_UnstableLiquidGun_Fire_Start_Close_02}}
 {id=WavePlayer_4, properties {soundWaveAssetPtr=&/WPN_UnstableLiquidGun_Fire_Start_Close_04}}
 {id=WavePlayer_5, properties {soundWaveAssetPtr=&/WPN_UnstableLiquidGun_Fire_Start_Close_05}}
 {
  id=WavePlayer_6
  properties {bLooping=true, soundWaveAssetPtr=&/WPN_UnstableLiquidGun_Fire_Loop_Close_01}
 }
 {
  id=WavePlayer_8
  properties {bLooping=true, soundWaveAssetPtr=&/WPN_UnstableLiquidGun_Fire_Loop_Close_02}
 }
]
properties {
 attenuationSettings=&Fort_Weapon_Projectile_FirstPerson_Athena_Att
 maxDistance=2097152
 soundClassObject=&Fort_WeaponRanged_Fire_1P_PSM
 volumeMultiplier=1.25
 attenuationOverrides {falloffDistance=5000, attenuationShapeExtents {x=1000}}
 concurrencyOverrides {bLimitToOwner=true, maxCount=3, resolutionRule=StopOldest}
 concurrencySet [&Weapon_Gunfire_Local_Conc, &Weapon_Gunfire_Global_Conc]
}
