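# First-person sound cue for the end of firing on the Unstable Liquid Gun.
# A Random node picks one of five equally weighted close-perspective fire-end waves, then a Modulator at the root applies pitch and volume modulation.
# It uses the first-person ranged-fire sound class, a shared first-person weapon projectile attenuation asset, and both local and global gunfire concurrency sets, at reduced volume.
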SoundCue /UnstableLiquidGun/Items/UnstableLiquidGun/Audio/UnstableLiquidGun_Fire_End_1P_Cue
root=Modulator_0
nodes [
 {id=Modulator_0, children [Random_0]}
 {id=Random_0, children [WavePlayer_0,WavePlayer_1,WavePlayer_2,WavePlayer_3,WavePlayer_4], properties {weights [1,1,1,1,1]}}
 {id=WavePlayer_0, properties {soundWaveAssetPtr=&/WPN_UnstableLiquidGun_Fire_End_Close_01}}
 {id=WavePlayer_1, properties {soundWaveAssetPtr=&/WPN_UnstableLiquidGun_Fire_End_Close_02}}
 {id=WavePlayer_2, properties {soundWaveAssetPtr=&/WPN_UnstableLiquidGun_Fire_End_Close_03}}
 {id=WavePlayer_3, properties {soundWaveAssetPtr=&/WPN_UnstableLiquidGun_Fire_End_Close_04}}
 {id=WavePlayer_4, properties {soundWaveAssetPtr=&/WPN_UnstableLiquidGun_Fire_End_Close_05}}
]
properties {
 attenuationSettings=&Fort_Weapon_Projectile_FirstPerson_Athena_Att
 maxDistance=2097152
 soundClassObject=&Fort_WeaponRanged_Fire_1P_PSM
 volumeMultiplier=0.7
 attenuationOverrides {falloffDistance=5000, attenuationShapeExtents {x=1000}}
 concurrencyOverrides {bLimitToOwner=true, maxCount=3, resolutionRule=StopOldest}
 concurrencySet [&Weapon_Gunfire_Local_Conc, &Weapon_Gunfire_Global_Conc]
}
