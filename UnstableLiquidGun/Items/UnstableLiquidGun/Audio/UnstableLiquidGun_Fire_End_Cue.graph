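# Sound cue for the end of the Unstable Liquid Gun's firing, layered by listener distance.
# A root mixer combines a close layer under a third-person fire sound class with a boosted medium-plus-distant pair under a distant fire class.
# Each layer picks randomly among equally weighted wave variants through a modulator and its own projectile attenuation.
# Local and global gunfire concurrency sets limit overlapping instances.
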SoundCue /UnstableLiquidGun/Items/UnstableLiquidGun/Audio/UnstableLiquidGun_Fire_End_Cue
root=Mixer_1
nodes [
 {
  id=Attenuation_10
  children [Modulator_14]
  properties {attenuationSettings=&Fort_Weapon_Projectile_Medium_Mono_Athena_Att}
 }
 {
  id=Attenuation_11
  children [Modulator_13]
  properties {attenuationSettings=&Fort_Weapon_Projectile_Distant_Mono_Athena_Att}
 }
 {id=Attenuation_9, children [Modulator_0], properties {attenuationSettings=&Fort_Weapon_Projectile_Close_Athena_Att}}
 {id=Mixer_1, children [SoundClass_7,SoundClass_6], properties {inputVolume [1,1.5]}}
 {id=Mixer_2, children [Attenuation_10, Attenuation_11], properties {inputVolume [1,1]}}
 {id=Modulator_0, children [Random_0]}
 {id=Modulator_13, children [Random_1], properties {pitchMax=1.1}}
 {id=Modulator_14, children [Random_2], properties {pitchMax=1.1, volumeMax=1, volumeMin=1}}
 {id=Random_0, children [WavePlayer_0,WavePlayer_1,WavePlayer_2,WavePlayer_3,WavePlayer_4], properties {weights [1,1,1,1,1]}}
 {id=Random_1, children [WavePlayer_5,WavePlayer_6,WavePlayer_7,WavePlayer_8,WavePlayer_9], properties {weights [1,1,1,1,1]}}
 {
  id=Random_2
  children [WavePlayer_10, WavePlayer_11, WavePlayer_12, WavePlayer_13, WavePlayer_14, WavePlayer_15]
  properties {weights [1,1,1,1,1,1]}
 }
 {id=SoundClass_6, children [Mixer_2], properties {soundClassOverride=&Fort_WeaponRanged_Fire_Distant}}
 {
  id=SoundClass_7
  children [Attenuation_9]
  properties {soundClassOverride=&Fort_WeaponRanged_Fire_3P_PSM}
 }
 {id=WavePlayer_0, properties {soundWaveAssetPtr=&/WPN_UnstableLiquidGun_Fire_End_Close_01}}
 {id=WavePlayer_1, properties {soundWaveAssetPtr=&/WPN_UnstableLiquidGun_Fire_End_Close_02}}
 {id=WavePlayer_10, properties {soundWaveAssetPtr=&/WPN_UnstableLiquidGun_Fire_End_Med_01}}
 {id=WavePlayer_11, properties {soundWaveAssetPtr=&/WPN_UnstableLiquidGun_Fire_End_Med_02}}
 {id=WavePlayer_12, properties {soundWaveAssetPtr=&/WPN_UnstableLiquidGun_Fire_End_Med_03}}
 {id=WavePlayer_13, properties {soundWaveAssetPtr=&/WPN_UnstableLiquidGun_Fire_End_Med_04}}
 {id=WavePlayer_14, properties {soundWaveAssetPtr=&/WPN_UnstableLiquidGun_Fire_End_Med_05}}
 {id=WavePlayer_15, properties {soundWaveAssetPtr=&/WPN_UnstableLiquidGun_Fire_End_Med_06}}
 {id=WavePlayer_2, properties {soundWaveAssetPtr=&/WPN_UnstableLiquidGun_Fire_End_Close_03}}
 {id=WavePlayer_3, properties {soundWaveAssetPtr=&/WPN_UnstableLiquidGun_Fire_End_Close_04}}
 {id=WavePlayer_4, properties {soundWaveAssetPtr=&/WPN_UnstableLiquidGun_Fire_End_Close_05}}
 {id=WavePlayer_5, properties {soundWaveAssetPtr=&/WPN_UnstableLiquidGun_Fire_End_Dist_01}}
 {id=WavePlayer_6, properties {soundWaveAssetPtr=&/WPN_UnstableLiquidGun_Fire_End_Dist_02}}
 {id=WavePlayer_7, properties {soundWaveAssetPtr=&/WPN_UnstableLiquidGun_Fire_End_Dist_03}}
 {id=WavePlayer_8, properties {soundWaveAssetPtr=&/WPN_UnstableLiquidGun_Fire_End_Dist_04}}
 {id=WavePlayer_9, properties {soundWaveAssetPtr=&/WPN_UnstableLiquidGun_Fire_End_Dist_05}}
]
properties {
 maxDistance=27500
 volumeMultiplier=0.5
 attenuationOverrides {falloffDistance=5000, attenuationShapeExtents {x=1000}}
 concurrencyOverrides {bLimitToOwner=true, maxCount=3, resolutionRule=StopOldest}
 concurrencySet [&Weapon_Gunfire_Local_Conc, &Weapon_Gunfire_Global_Conc]
}
