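# Sound cue that plays when the Unstable Liquid Gun weapon is picked up.
# A single wave player feeds a modulator at the root.
# It uses the Fort weapon pickup attenuation, the loot collect sound class and a shared pickup concurrency set, alongside a one-instance concurrency override that prevents new plays.
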SoundCue /UnstableLiquidGun/Items/UnstableLiquidGun/Audio/UnstableLiquidGun_Pickup_Cue
root=Modulator_0
nodes [
 {id=Modulator_0, children [WavePlayer_0]}
 {id=WavePlayer_0, properties {soundWaveAssetPtr=&/WPN_UnstableLiquidGun_Pickup_01}}
]
properties {
 attenuationSettings=&Fort_Weapon_Pickup_Att
 maxDistance=3500
 soundClassObject=&Fort_Loot_Collect
 volumeMultiplier=0.796
 concurrencyOverrides {maxCount=1, resolutionRule=PreventNew}
 concurrencySet [&PickupMaximum_Oldest]
}
