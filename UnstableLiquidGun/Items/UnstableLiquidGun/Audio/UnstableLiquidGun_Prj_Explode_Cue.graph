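# Sound cue for the Unstable Liquid Gun projectile explosion, mixing close, medium and distant layers, each with its own attenuation and sound class.
# A quality-level node gives the first two tiers full random wave pools and the last two trimmed three-wave pools without water handling.
# In the full tiers, InWater switches add delayed, pitch-modulated water debris to close and medium layers.
# Concurrency caps it at three instances, stopping the oldest.
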SoundCue /UnstableLiquidGun/Items/UnstableLiquidGun/Audio/UnstableLiquidGun_Prj_Explode_Cue
root=QualityLevel_0
nodes [
 {
  id=Attenuation_1
  children [Modulator_1]
  properties {
   attenuationSettings=&/UnstableLiquidGun_Proj_Expl_Close_Att
   attenuationOverrides {falloffDistance=4000, nonSpatializedRadiusStart=500, reverbWetLevelMin=0.2, stereoSpread=5000, attenuationShapeExtents {x=2000}}
  }
 }
 {
  id=Attenuation_2
  children [Switch_1]
  properties {
   attenuationSettings=&/UnstableLiquidGun_Proj_Expl_Med_Att
   attenuationOverrides {falloffDistance=3000, attenuationShapeExtents {x=1000}}
  }
 }
 {
  id=Attenuation_3
  children [Random_4]
  properties {
   attenuationSettings=&/UnstableLiquidGun_Proj_Expl_Dist_Att
   attenuationOverrides {falloffDistance=3000, attenuationShapeExtents {x=1000}}
  }
 }
 {
  id=Attenuation_6
  children [Modulator_2]
  properties {
   attenuationSettings=&UnstableLiquidGun_Proj_Expl_Close_Att
   attenuationOverrides {falloffDistance=4000, nonSpatializedRadiusStart=500, reverbWetLevelMin=0.2, stereoSpread=5000, attenuationShapeExtents {x=2000}}
  }
 }
 {
  id=Attenuation_7
  children [Random_6]
  properties {
   attenuationSettings=&UnstableLiquidGun_Proj_Expl_Med_Att
   attenuationOverrides {falloffDistance=3000, attenuationShapeExtents {x=1000}}
  }
 }
 {
  id=Attenuation_8
  children [Random_7]
  properties {
   attenuationSettings=&UnstableLiquidGun_Proj_Expl_Dist_Att
   attenuationOverrides {falloffDistance=3000, attenuationShapeExtents {x=1000}}
  }
 }
 {id=Delay_0, children [Modulator_0], properties {delayMax=0.2, delayMin=0.1}}
 {id=Delay_1, children [Modulator_3], properties {delayMax=0.2, delayMin=0.1}}
 {
  id=Enveloper_0
  children [Random_2]
  properties {
   pitchCurve {editorCurveData {keys [{time=0}]}}
   volumeCurve {editorCurveData {keys [{arriveTangent=-1.2286, interpMode=RCIM_Cubic, leaveTangent=-1.2286, tangentMode=RCTM_User, time=0, value=1}, <at 2 0 RCIM_Cubic>]}}
  }
 }
 {id=Mixer_0, children [Random_1,Modulator_4], properties {inputVolume [1,0.8]}}
 {id=Mixer_1, children [SoundClass_1,SoundClass_4,SoundClass_5], properties {inputVolume [1,1,1]}}
 {id=Mixer_2, children [SoundClass_6,SoundClass_7,SoundClass_8], properties {inputVolume [1,1,1]}}
 {id=Mixer_3, children [Random_3,Delay_1], properties {inputVolume [1,0.8]}}
 {id=Modulator_0, children [Enveloper_0], properties {pitchMax=1.3, pitchMin=1.1, volumeMax=1.1, volumeMin=0.9}}
 {id=Modulator_1, children [Switch_0], properties {pitchMax=1.2}}
 {id=Modulator_2, children [Random_5], properties {pitchMax=1.2}}
 {id=Modulator_3, children [Random_0], properties {pitchMax=1.2, pitchMin=1, volumeMax=1.1, volumeMin=0.9}}
 {id=Modulator_4, children [Delay_0], properties {pitchMax=1.2, pitchMin=1.1}}
 {id=QualityLevel_0, children [Mixer_1,Mixer_1,Mixer_2,Mixer_2]}
 {id=Random_0, children [WavePlayer_25, WavePlayer_26, WavePlayer_29, WavePlayer_30], properties {weights [1,1,1,1]}}
 {id=Random_1, children [WavePlayer_3, WavePlayer_5, WavePlayer_6, WavePlayer_7, WavePlayer_10], properties {weights [1,1,1,1,1]}}
 {id=Random_2, children [WavePlayer_4, WavePlayer_8, WavePlayer_9, WavePlayer_22], properties {weights [1,1,1,1]}}
 {id=Random_3, children [WavePlayer_11, WavePlayer_12, WavePlayer_13, WavePlayer_14, WavePlayer_15], properties {weights [1,1,1,1,1]}}
 {
  id=Random_4
  children [WavePlayer_16, WavePlayer_17, WavePlayer_18, WavePlayer_19, WavePlayer_20, WavePlayer_21]
  properties {weights [1,1,1,1,1,1]}
 }
 {id=Random_5, children [WavePlayer_23, WavePlayer_24, WavePlayer_0], properties {bShouldExcludeFromBranchCulling=true, weights [1,1,1]}}
 {id=Random_6, children [WavePlayer_27, WavePlayer_28, WavePlayer_1], properties {bShouldExcludeFromBranchCulling=true, weights [1,1,1]}}
 {id=Random_7, children [WavePlayer_33, WavePlayer_34, WavePlayer_2], properties {bShouldExcludeFromBranchCulling=true, weights [1,1,1]}}
 {id=SoundClass_1, children [Attenuation_1], properties {soundClassOverride=/Game/Sounds/FortSoundClassesAndModes/Fort_Explosives_Large_PSM}}
 {
  id=SoundClass_4
  children [Attenuation_2]
  properties {soundClassOverride=&Fort_WeaponRanged_Fire_3P_PSM}
 }
 {
  id=SoundClass_5
  children [Attenuation_3]
  properties {soundClassOverride=&Fort_WeaponRanged_Fire_Distant}
 }
 {id=SoundClass_6, children [Attenuation_6], properties {soundClassOverride=&Fort_Explosives_Large_PSM}}
 {
  id=SoundClass_7
  children [Attenuation_7]
  properties {soundClassOverride=&Fort_WeaponRanged_Fire_3P_PSM}
 }
 {
  id=SoundClass_8
  children [Attenuation_8]
  properties {soundClassOverride=&Fort_WeaponRanged_Fire_Distant}
 }
 {id=Switch_0, children [Random_1,Random_1,Mixer_0], properties {intParameterName=InWater}}
 {id=Switch_1, children [Random_3,Random_3,Mixer_3], properties {intParameterName=InWater}}
 {id=WavePlayer_0, properties {soundWaveAssetPtr=&/WPN_UnstableLiquidGun_Proj_Expl_Close_03}}
 {id=WavePlayer_1, properties {soundWaveAssetPtr=&/WPN_UnstableLiquidGun_Proj_Expl_Med_03}}
 {id=WavePlayer_10, properties {soundWaveAssetPtr=&/WPN_UnstableLiquidGun_Proj_Expl_Close_04}}
 {id=WavePlayer_11, properties {soundWaveAssetPtr=&/WPN_UnstableLiquidGun_Proj_Expl_Med_01}}
 {id=WavePlayer_12, properties {soundWaveAssetPtr=&/WPN_UnstableLiquidGun_Proj_Expl_Med_02}}
 {id=WavePlayer_13, properties {soundWaveAssetPtr=&WPN_UnstableLiquidGun_Proj_Expl_Med_03}}
 {id=WavePlayer_14, properties {soundWaveAssetPtr=&/WPN_UnstableLiquidGun_Proj_Expl_Med_04}}
 {id=WavePlayer_15, properties {soundWaveAssetPtr=&/WPN_UnstableLiquidGun_Proj_Expl_Med_05}}
 {id=WavePlayer_16, properties {soundWaveAssetPtr=&/WPN_UnstableLiquidGun_Proj_Expl_Dist_06}}
 {id=WavePlayer_17, properties {soundWaveAssetPtr=&/WPN_UnstableLiquidGun_Proj_Expl_Dist_01}}
 {id=WavePlayer_18, properties {soundWaveAssetPtr=&/WPN_UnstableLiquidGun_Proj_Expl_Dist_02}}
 {id=WavePlayer_19, properties {soundWaveAssetPtr=&/WPN_UnstableLiquidGun_Proj_Expl_Dist_03}}
 {id=WavePlayer_2, properties {soundWaveAssetPtr=&WPN_UnstableLiquidGun_Proj_Expl_Dist_03}}
 {id=WavePlayer_20, properties {soundWaveAssetPtr=&/WPN_UnstableLiquidGun_Proj_Expl_Dist_04}}
 {id=WavePlayer_21, properties {soundWaveAssetPtr=&/WPN_UnstableLiquidGun_Proj_Expl_Dist_05}}
 {id=WavePlayer_22, properties {soundWaveAssetPtr=/Game/Athena/Sounds/Weapons/Explosions/EXPL_Debris_Water_Wave_04}}
 {id=WavePlayer_23, properties {soundWaveAssetPtr=&/WPN_UnstableLiquidGun_Proj_Expl_Close_01}}
 {id=WavePlayer_24, properties {soundWaveAssetPtr=&/WPN_UnstableLiquidGun_Proj_Expl_Close_02}}
 {id=WavePlayer_25, properties {soundWaveAssetPtr=/Game/Athena/Sounds/Weapons/Explosions/EXPL_Debris_Water_Wave_01}}
 {id=WavePlayer_26, properties {soundWaveAssetPtr=/Game/Athena/Sounds/Weapons/Explosions/EXPL_Debris_Water_Wave_02}}
 {id=WavePlayer_27, properties {soundWaveAssetPtr=&WPN_UnstableLiquidGun_Proj_Expl_Med_01}}
 {id=WavePlayer_28, properties {soundWaveAssetPtr=&WPN_UnstableLiquidGun_Proj_Expl_Med_02}}
 {id=WavePlayer_29, properties {soundWaveAssetPtr=/Game/Athena/Sounds/Weapons/Explosions/EXPL_Debris_Water_Wave_03}}
 {id=WavePlayer_3, properties {soundWaveAssetPtr=&/WPN_UnstableLiquidGun_Proj_Expl_Close_05}}
 {id=WavePlayer_30, properties {soundWaveAssetPtr=&EXPL_Debris_Water_Wave_04}}
 {id=WavePlayer_33, properties {soundWaveAssetPtr=&WPN_UnstableLiquidGun_Proj_Expl_Dist_01}}
 {id=WavePlayer_34, properties {soundWaveAssetPtr=&WPN_UnstableLiquidGun_Proj_Expl_Dist_02}}
 {id=WavePlayer_4, properties {soundWaveAssetPtr=&EXPL_Debris_Water_Wave_01}}
 {id=WavePlayer_5, properties {soundWaveAssetPtr=&WPN_UnstableLiquidGun_Proj_Expl_Close_01}}
 {id=WavePlayer_6, properties {soundWaveAssetPtr=&WPN_UnstableLiquidGun_Proj_Expl_Close_02}}
 {id=WavePlayer_7, properties {soundWaveAssetPtr=&WPN_UnstableLiquidGun_Proj_Expl_Close_03}}
 {id=WavePlayer_8, properties {soundWaveAssetPtr=&EXPL_Debris_Water_Wave_02}}
 {id=WavePlayer_9, properties {soundWaveAssetPtr=&EXPL_Debris_Water_Wave_03}}
]
properties {
 bOverrideConcurrency=true
 maxDistance=27500
 soundClassObject=&Fort_Explosives_Large_PSM
 volumeMultiplier=0.5307
 attenuationOverrides {falloffDistance=7500}
 concurrencyOverrides {maxCount=3, resolutionRule=StopOldest, retriggerTime=0.1}
}
