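# Sound cue for an Unstable Liquid Gun projectile landing, playing a randomized, attenuated impact sound.
# A quality switch sends its first two levels to a random pick of six waves and its last two to a pick of three, exempt from branch culling.
# Each branch passes through a pitch modulator for variation.
# A concurrency override caps simultaneous plays, stopping the oldest and scaling volume to tame rapid impacts.
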SoundCue /UnstableLiquidGun/Items/UnstableLiquidGun/Audio/UnstableLiquidGun_Prj_Land_Cue
root=QualityLevel_0
nodes [
 {id=Modulator_0, children [Random_2], properties {pitchMax=1.2}}
 {id=Modulator_1, children [Random_3], properties {pitchMax=1.2}}
 {id=QualityLevel_0, children [Modulator_0,Modulator_0,Modulator_1,Modulator_1]}
 {id=Random_2, children [WavePlayer_0,WavePlayer_1,WavePlayer_2,WavePlayer_3,WavePlayer_7,WavePlayer_8], properties {weights [1,1,1,1,1,1]}}
 {id=Random_3, children [WavePlayer_9, WavePlayer_10, WavePlayer_4], properties {bShouldExcludeFromBranchCulling=true, weights [1,1,1]}}
 {id=WavePlayer_0, properties {soundWaveAssetPtr=&/WPN_UnstableLiquidGun_Proj_Impact_01}}
 {id=WavePlayer_1, properties {soundWaveAssetPtr=&/WPN_UnstableLiquidGun_Proj_Impact_02}}
 {id=WavePlayer_10, properties {soundWaveAssetPtr=&WPN_UnstableLiquidGun_Proj_Impact_02}}
 {id=WavePlayer_2, properties {soundWaveAssetPtr=&/WPN_UnstableLiquidGun_Proj_Impact_03}}
 {id=WavePlayer_3, properties {soundWaveAssetPtr=&/WPN_UnstableLiquidGun_Proj_Impact_04}}
 {id=WavePlayer_4, properties {soundWaveAssetPtr=&WPN_UnstableLiquidGun_Proj_Impact_03}}
 {id=WavePlayer_7, properties {soundWaveAssetPtr=&/WPN_UnstableLiquidGun_Proj_Impact_05}}
 {id=WavePlayer_8, properties {soundWaveAssetPtr=&/WPN_UnstableLiquidGun_Proj_Impact_06}}
 {id=WavePlayer_9, properties {soundWaveAssetPtr=&WPN_UnstableLiquidGun_Proj_Impact_01}}
]
properties {
 attenuationSettings=&/UnstableLiquidGun_Proj_Impact_Att
 bOverrideConcurrency=true
 maxDistance=5500
 soundClassObject=&Fort_WeaponRanged_Fire_3P_PSM
 volumeMultiplier=0.5
 concurrencyOverrides {maxCount=2, resolutionRule=StopOldest, retriggerTime=0.2, volumeScale=0.7, volumeScaleAttackTime=0.2}
}
