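# Sound cue for the fly-by sound of Unstable Liquid Gun projectiles, classed as a ranged weapon sound.
# It picks one of six equally weighted pass-by waves and feeds it through a modulator that randomizes volume and slightly lowers pitch by a varying amount.
# Spatial falloff references a dedicated fly-by attenuation asset.
# Overridden concurrency prevents new instances beyond a small cap, with a retrigger delay and volume scaling to limit stacking.
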SoundCue /UnstableLiquidGun/Items/UnstableLiquidGun/Audio/UnstableLiquidGun_Proj_FlyBy_Cue
root=Modulator_0
nodes [
 {id=Modulator_0, children [Random_0], properties {pitchMax=1, pitchMin=0.8, volumeMax=0.6, volumeMin=0.5}}
 {
  id=Random_0
  children [WavePlayer_1, WavePlayer_2, WavePlayer_3, WavePlayer_4, WavePlayer_5, WavePlayer_10]
  properties {weights [1,1,1,1,1,1]}
 }
 {id=WavePlayer_1, properties {soundWaveAssetPtr=&/WPN_UnstableLiquidGun_Proj_PassBy_01}}
 {id=WavePlayer_10, properties {soundWaveAssetPtr=&/WPN_UnstableLiquidGun_Proj_PassBy_06}}
 {id=WavePlayer_2, properties {soundWaveAssetPtr=&/WPN_UnstableLiquidGun_Proj_PassBy_02}}
 {id=WavePlayer_3, properties {soundWaveAssetPtr=&/WPN_UnstableLiquidGun_Proj_PassBy_03}}
 {id=WavePlayer_4, properties {soundWaveAssetPtr=&/WPN_UnstableLiquidGun_Proj_PassBy_04}}
 {id=WavePlayer_5, properties {soundWaveAssetPtr=&/WPN_UnstableLiquidGun_Proj_PassBy_05}}
]
properties {
 attenuationSettings=&/UnstableLiquidGun_Proj_FlyBy_Att
 bOverrideConcurrency=true
 maxDistance=4500
 soundClassObject=&Fort_WeaponRanged
 volumeMultiplier=2
 attenuationOverrides {distanceAlgorithm=NaturalSound, falloffDistance=2500, attenuationShapeExtents {x=500}}
 concurrencyOverrides {maxCount=2, resolutionRule=PreventNew, retriggerTime=0.3, volumeScale=0.75, volumeScaleAttackTime=0.2}
}
