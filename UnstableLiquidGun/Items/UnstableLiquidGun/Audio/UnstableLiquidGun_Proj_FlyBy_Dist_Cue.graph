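# Sound cue for the distant fly-by of Unstable Liquid Gun projectiles, in the ranged weapon sound class.
# A Random node picks one of six equally weighted pass-by waves, and a Modulator varies its pitch and volume.
# A distant fly-by attenuation asset handles falloff.
# Overridden concurrency caps simultaneous plays, blocks new ones at the limit, and adds retrigger spacing and volume scaling.
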SoundCue /UnstableLiquidGun/Items/UnstableLiquidGun/Audio/UnstableLiquidGun_Proj_FlyBy_Dist_Cue
root=Modulator_0
nodes [
 {id=Modulator_0, children [Random_1], properties {pitchMax=1, pitchMin=0.8, volumeMax=1, volumeMin=0.7}}
 {
  id=Random_1
  children [WavePlayer_6, WavePlayer_7, WavePlayer_8, WavePlayer_9, WavePlayer_11, WavePlayer_12]
  properties {weights [1,1,1,1,1,1]}
 }
 {id=WavePlayer_11, properties {soundWaveAssetPtr=&/WPN_UnstableLiquidGun_Proj_PassBy_Dist_05}}
 {id=WavePlayer_12, properties {soundWaveAssetPtr=&/WPN_UnstableLiquidGun_Proj_PassBy_Dist_06}}
 {id=WavePlayer_6, properties {soundWaveAssetPtr=&/WPN_UnstableLiquidGun_Proj_PassBy_Dist_01}}
 {id=WavePlayer_7, properties {soundWaveAssetPtr=&/WPN_UnstableLiquidGun_Proj_PassBy_Dist_02}}
 {id=WavePlayer_8, properties {soundWaveAssetPtr=&/WPN_UnstableLiquidGun_Proj_PassBy_Dist_03}}
 {id=WavePlayer_9, properties {soundWaveAssetPtr=&/WPN_UnstableLiquidGun_Proj_PassBy_Dist_04}}
]
properties {
 attenuationSettings=&/UnstableLiquidGun_Proj_FlyBy_Dist_Att
 bOverrideConcurrency=true
 maxDistance=7000
 soundClassObject=&Fort_WeaponRanged
 volumeMultiplier=2
 attenuationOverrides {distanceAlgorithm=NaturalSound, falloffDistance=2500, attenuationShapeExtents {x=500}}
 concurrencyOverrides {maxCount=2, resolutionRule=PreventNew, retriggerTime=0.3, volumeScale=0.75, volumeScaleAttackTime=0.2}
}
