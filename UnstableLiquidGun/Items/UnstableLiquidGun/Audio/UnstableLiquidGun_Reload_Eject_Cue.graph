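# Sound cue for the eject sound of the Unstable Liquid Gun's reload.
# A single wave player feeds a modulator left at default settings, with the cue's volume lowered to about half.
# It uses the shared weapon reload attenuation, the weapon foley sound class and a concurrency set named for pickups, PickupMaximum_Oldest.
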SoundCue /UnstableLiquidGun/Items/UnstableLiquidGun/Audio/UnstableLiquidGun_Reload_Eject_Cue
root=Modulator_0
nodes [
 {id=Modulator_0, children [WavePlayer_0]}
 {id=WavePlayer_0, properties {soundWaveAssetPtr=&/WPN_UnstableLiquidGun_Reload_Eject}}
]
properties {
 attenuationSettings=&Fort_Weapon_Reload_Att
 maxDistance=2250
 soundClassObject=&Fort_WeaponFoley
 volumeMultiplier=0.501
 concurrencyOverrides {maxCount=1, resolutionRule=PreventNew}
 concurrencySet [&PickupMaximum_Oldest]
}
