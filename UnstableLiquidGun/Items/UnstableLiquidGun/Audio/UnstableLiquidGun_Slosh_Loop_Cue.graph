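# Looping foley cue for the Unstable Liquid Gun's tank slosh, silenced when the Visible parameter is false.
# Movement scales the overall volume, defaulting to silent.
# PistonProgress crossfades from a tank-low loop to a tank-full loop and modulates the full loop's pitch.
# Two chained modulators shape the low loop into a volume bump that fades in from zero progress and out as the full loop rises.
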SoundCue /UnstableLiquidGun/Items/UnstableLiquidGun/Audio/UnstableLiquidGun_Slosh_Loop_Cue
root=Branch_0
nodes [
 {id=Branch_0, children [ModulatorContinuous_2, null, ModulatorContinuous_2], properties {boolParameterName=Visible}}
 {id=Mixer_0, children [ModulatorContinuous_0, ModulatorContinuous_3], properties {inputVolume [1,1]}}
 {
  id=ModulatorContinuous_0
  children [WavePlayer_3]
  properties {
   pitchModulationParams {minInput=0.25, minOutput=1.2, paramMode=MPM_Abs, parameterName=PistonProgress}
   volumeModulationParams {maxInput=0.5, minInput=0.25, paramMode=MPM_Abs, parameterName=PistonProgress}
  }
 }
 {
  id=ModulatorContinuous_1
  children [WavePlayer_4]
  properties {volumeModulationParams {maxInput=0.5, maxOutput=0, minInput=0.25, minOutput=1, paramMode=MPM_Abs, parameterName=PistonProgress}}
 }
 {id=ModulatorContinuous_2, children [Mixer_0], properties {volumeModulationParams {default=0, maxOutput=0.5, parameterName=Movement}}}
 {
  id=ModulatorContinuous_3
  children [ModulatorContinuous_1]
  properties {volumeModulationParams {maxInput=0.25, paramMode=MPM_Abs, parameterName=PistonProgress}}
 }
 {
  id=WavePlayer_3
  properties {bLooping=true, soundWaveAssetPtr=&/WPN_UnstableLiquidGun_Slosh_TankFull_Loop_01}
 }
 {
  id=WavePlayer_4
  properties {bLooping=true, soundWaveAssetPtr=&/WPN_UnstableLiquidGun_Slosh_TankLow_Loop_01}
 }
]
properties {
 attenuationSettings=&Fort_Weapon_Reload_Att
 bDebug=true
 bOverrideConcurrency=true
 maxDistance=2250
 soundClassObject=&Fort_WeaponFoley
 volumeMultiplier=0.2
}
