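; BP_LiquidGun_RibbonManager draws the liquid stream of the Unstable Liquid Gun as a Niagara ribbon through its fired projectiles.
; Each tick it computes spline points, thickness and flashing intensity, natively or through a Blueprint fallback, and pushes them to Niagara.
; Over-stretched or angle-broken segments in the Blueprint path split into separate ribbons with curling, tapered ends.
; The server runs spline collision tests, and cleanup destroys projectiles and the manager, notifying the weapon if mid-stream.
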(event EventBeginPlay
  ;[1 1 1] Spawn Ribbon effect on begin play
  (Variables|Default|SetBeginPlayTime (Utilities|Time|GetGameTimeinSeconds))
  (bind _returnvalue (Niagara|SpawnSystematLocation "/UnstableLiquidGun/Items/UnstableLiquidGun/FX/NS_UnstableLiquidGun_Ribbon_V2.NS_UnstableLiquidGun_Ribbon_V2" (Transformation|GetActorLocation)))
  (bind _output_get (Variables|Default|SetRibbonFX _returnvalue))
  (Rendering|SetCastShadow _output_get true))

(event Custom|CE_StopExplosion
  (bind _ribbonfx (Variables|Default|GetRibbonFX)
    (:then
      (Effects|Components|ParticleSystem|SetFloatParameter _ribbonfx "BubbleSpawnRate" 0.0)
      (Utilities|FlowControl|Delay 0.6)
      (bind _ribbonfx_2 (Variables|Default|GetRibbonFX)
        (:then
          (Components|Activation|Deactivate _ribbonfx_2))
        (:else)))
    (:else)))

(event Custom|CleanUp
  (bind _ribbonfx (Variables|Default|GetRibbonFX))
  ;[1 1 1] Destroy Self
  (Variables|Default|SetDone true)
  (for _array_element (Variables|Default|GetProjectiles)
    (Utilities|IsValid _array_element
      (:Is%20Valid
        (Actor|DestroyActor _array_element))))
  (Components|DestroyComponent _ribbonfx)
  (Development|SetHiddeninGame _ribbonfx true false)
  (Variables|Default|GetWeapon
    (:then
      (if (and (Variables|/UnstableLiquidGun/Items/UnstableLiquidGun/B_Weapon_LiquidGun.B_Weapon_LiquidGun_C|GetMidStream (Variables|Default|GetWeapon)) (== (Variables|/UnstableLiquidGun/Items/UnstableLiquidGun/B_Weapon_LiquidGun.B_Weapon_LiquidGun_C|GetCurrentManager (Variables|Default|GetWeapon)) self))
        (Class|BWeaponLiquidGun|HandleManagerCleanUpMidStream (Variables|Default|GetWeapon)) ; It's unlikely but possible for the explosions to reach back to the gun, meaning we clear everything while the player is firing.\nThis forces the player to stop firing in that scenario.
        (else
          _))
      (Utilities|FlowControl|Delay 1.5)
      (Actor|DestroyActor))
    (:else
      _)))

(event Custom|AddProjectile
  (params (Prj :type B%20Prj%20Ranged%20Nitroglycerine%20Gun%20Object%20Reference%20%28by%20ref%29))
  (bind _self self)
  (bind _waterinteractioncomponent (Variables|/Script/FortniteGame.FortProjectileBase|GetWaterInteractionComponent Prj))
  ;[1 1 1] Add recently created projectile (called by weapon)
  (bind _returnvalue (Utilities|Array|Add (Variables|Default|GetProjectiles) Prj))
  ;[0.4299 0.8626 1] Set both LiquidRibbonManager RibbonManager (used in code) and BP_LiquidGun_RibbonManager CachedBPRibbonManager so RibbonManager doesn't have to be cast each time in BP
  (Variables|/Script/UnstableLiquidGunRuntime.FortLiquidGunProjectile|SetRibbonManager _self Prj)
  (Variables|/UnstableLiquidGun/Items/UnstableLiquidGun/B_Prj_Ranged_NitroglycerineGun.B_Prj_Ranged_NitroglycerineGun_C|SetCachedBPRibbonManager _self Prj)
  (bind _output_get (Variables|/Script/UnstableLiquidGunRuntime.FortLiquidGunProjectile|SetProjectileIndex _returnvalue Prj))
  (if (== _output_get 1)
    (EventDispatchers|BindEventtoOnStopped Prj (delegate Custom|FirstProjectileStopped%20%28server%20only%29))
    ;[0.4135 0.8072 1] Consider entering water water a stop event for these purposes
    (Utilities|IsValid _waterinteractioncomponent
      (:Is%20Valid
        (if (Water|IsinWater _waterinteractioncomponent)
          (CallFunction|FirstProjectileStopped(ServerOnly))
          (else
            (Water|BindEventtoOnEnterWaterEvent _waterinteractioncomponent (delegate Custom|FirstProjectileEnteredWater))))))))

(event Custom|FirstProjectileStopped%20%28server%20only%29)

(event Custom|FirstProjectileEnteredWater
  (params (WaterBody :type Fort%20Water%20Body%20Actor%20Object%20Reference) (WaterInteractionComponent :type Fort%20Water%20Interaction%20Component%20Object%20Reference) (bIsFirstBody :type Boolean))
  ;[0.4135 0.8072 1] Consider entering water water a stop event for these purposes
  (CallFunction|FirstProjectileStopped(ServerOnly)))

(event EventTick
  (params (DeltaSeconds :type Float%20%28single-precision%29))
  ;[0.5384 1 0.6491] Tick
  (|CallParentFunction DeltaSeconds)
  (Niagara|SetNiagaraVariable(Int32) (Variables|Default|GetRibbonFX) "PreExplodeIndex" (+ (Math|Float|Truncate (/ (Utilities|TimeSinceBeginPlay) 0.0083)) 100))
  (if (not (Variables|Default|GetDone))
    (Utilities|FlowControl|Sequence
      (:then_0
        (CallFunction|UpdateNiagaraPositions false :!pinned true) ; Meat of functionality is in this function; calculates the curve and updates the ribbon
      )
      (:then_1
        (Networking|SwitchHasAuthority
          (:Authority
            (CallFunction|RunSplineCollisionTestonProjectiles))) ; For perf, run the spline collision on the server only.\nMay need to predictively run on the client if this feels too delayed 
      ))))

(fn AddDetachedSegments
  (params (Positions :type Array%20of%20Vectors) (RibbonIDs :type Array%20of%20Integers))
  (bind _currribbonnumber (Variables|Default|GetCurrRibbonNumber))
  (LocalVariables|SetCurrentRibbon _currribbonnumber)
  (for _index (range (Utilities|Array|FindItem RibbonIDs _currribbonnumber) (Utilities|Array|Length RibbonIDs))
    (if (== (Utilities|Array|Get(acopy) (Variables|GetRibbonIDs) _index) (LocalVariables|GetCurrentRibbon))
      (Utilities|Array|Add (LocalVariables|GetRibbonPositions) (Utilities|Array|Get(acopy) (Variables|GetPositions) _index))))
  (Utilities|Array|Add (Variables|Default|GetDetachedRibbons) (Utilities|Struct|MakeLiquidGunDetachedRibbons (LocalVariables|GetRibbonPositions) (Utilities|Array|Length (LocalVariables|GetRibbonPositions)) (LocalVariables|GetCurrentRibbon) (Utilities|Time|GetGameTimeinSeconds)))
  (Math|Integer|IncrementInt (Variables|Default|GetCurrRibbonNumber)))

(fn BPGetOrAddDetachment
  (params (RibbonNumber :type Integer) (Segment :type Liquid%20Ribbon%20Spline%20Segment%20Structure))
  (bind _returnvalue (CallFunction|BPGetPointonSegment 0.0 Segment))
  (bind _returnvalue_2 (CallFunction|BPGetPointonSegment 1.0 Segment))
  (bind _returnvalue_3 (Utilities|Time|GetGameTimeinSeconds))
  (bind _returnvalue_4 (+ RibbonNumber 1))
  (bind _liquidgun_detachment (Utilities|Struct|MakeLiquidGunDetachment RibbonNumber _returnvalue_4 _returnvalue_3 Segment _returnvalue _returnvalue_2))
  (if (< RibbonNumber (Utilities|Array|Length (Variables|Default|GetBP_Detachments)))
    (bind _structout (Utilities|Struct|SetmembersinLiquidGunDetachment (Utilities|Array|Get(aref) (Variables|Default|GetBP_Detachments) RibbonNumber) Segment (CallFunction|BPGetPointonSegment 0.0 Segment) (CallFunction|BPGetPointonSegment 1.0 Segment)))
    (return _structout)
    (else
      (Utilities|Array|Add (Variables|Default|GetBP_Detachments) _liquidgun_detachment)
      (return _liquidgun_detachment))))

(fn BPGetPointOnSegment
  (params (t :type Float%20%28double-precision%29) (segment :type Liquid%20Ribbon%20Spline%20Segment%20Structure))
  (bind (_a _b _c _d) (Utilities|Struct|BreakLiquidRibbonSplineSegment segment))
  (return (Utilities|Operators|Add (Utilities|Operators|Multiply _a t t t) (Utilities|Operators|Multiply _b t t) (* _c t) _d)))

(fn BPOVCalculatePointBlinkingIntensity
  (params (Point%20Size :type Integer))
  (return (Utilities|RemapFloat (Utilities|RemapFloat (Math|Conversions|ToFloat(Integer) Point%20Size) (Math|Vector2D|MakeVector2D (Variables|PointSize|GetMinPointSize) (Variables|PointSize|GetMaxPointSize)) "(X=0.000000,Y=100.000000)" true) (Math|Vector2D|MakeVector2D (- (Variables|Default|GetBP%20%20DamagePercentForSizeIncrease) 2.0) (+ (Variables|Default|GetBP%20%20DamagePercentForSizeIncrease) 2.0)) "(X=0.000000,Y=1.000000)" true)))

(fn BPOVCalculatePointSize
  (params (SegmentIndex :type Integer) (SegmentLengths :type Array%20of%20Float%20%28double-precision%29s) (PointIndex :type Integer) (PointsPerSegment :type Integer) (Min%20Thickness :type Float%20%28double-precision%29) (Max%20Thickness :type Float%20%28double-precision%29))
  (bind _maxsegmentlengthbeforebreak (Variables|LiquidRibbonManager|GetMaxSegmentLengthBeforeBreak))
  (bind _returnvalue (ScalableFloat|GetValueatLevel _maxsegmentlengthbeforebreak))
  (bind _splinelengthforsizecalculation (Variables|PointSize|GetSplineLengthForSizeCalculation))
  (LocalVariables|SetPointsConsumed 0)
  (LocalVariables|SetPooledLength 0.0)
  (bind _index (Utilities|FlowControl|ForLoopwithBreak 0 SegmentIndex)
    (:LoopBody
      (LocalVariables|SetPointsConsumed (+ (LocalVariables|GetPointsConsumed) (Variables|LiquidRibbonManager|GetPointsPerSegment)))
      (bind _output_get (LocalVariables|SetPooledLength (+ (Math|Float|Min(Float) _returnvalue (Utilities|Array|Get(acopy) SegmentLengths (- SegmentIndex _index))) (LocalVariables|GetPooledLength))))
      (if (> _output_get _splinelengthforsizecalculation)
        (while (> _output_get _splinelengthforsizecalculation)
          (LocalVariables|SetPooledLength (- (LocalVariables|GetPooledLength) (/ (Math|Float|Min(Float) _returnvalue (Utilities|Array|Get(acopy) SegmentLengths (- SegmentIndex _index))) (Variables|LiquidRibbonManager|GetPointsPerSegment))))
          (Math|Integer|DecrementInt (LocalVariables|GetPointsConsumed)))))
    (:Completed
      _))
  (return (Utilities|RemapFloat (Math|Conversions|ToFloat(Integer) (LocalVariables|GetPointsConsumed)) "(X=1.000000,Y=50.000000)" (Math|Vector2D|MakeVector2D Min%20Thickness Max%20Thickness) true)))

(fn BP_GetFirstVisibleIndex
  (return 0))

(fn BP_PointIsVisible
  (params (CurrentPointIndex :type Integer) (FirstVisibleIndex :type Integer))
  (return true))

(fn CalculateOffset
  (params (Detachment :type Liquid%20Gun%20Detachment%20Structure) (Projectile :type B%20Prj%20Ranged%20Nitroglycerine%20Gun%20Object%20Reference) (ClosenessToBreakage :type Float%20%28double-precision%29) (IsBeginning :type Boolean))
  (bind (_firstribbonid_6_d38b49c548211254b7303e91cc87c3d0 _secondribbonid_7_5dcd0da94598430c7a3ae389f35c563c _timewhendetached_10_466cca0f4e3b02acaf08c39264b1976a _segment_25_d2b36ed04b9565d0469a36b1ba847ab6 _segmentbeginning_16_66b4e98b4755ad97ca3f179663bb8a22 _segmentend_18_2208cfe047d7b1790a1862865f202f17) (Utilities|Struct|BreakLiquidGunDetachment Detachment))
  (bind (_a _b _c _d _directlength) (Utilities|Struct|BreakLiquidRibbonSplineSegment _segment_25_d2b36ed04b9565d0469a36b1ba847ab6))
  (bind _returnvalue (- _segmentend_18_2208cfe047d7b1790a1862865f202f17 _segmentbeginning_16_66b4e98b4755ad97ca3f179663bb8a22))
  (if (Class|BPrjRangedNitroglycerineGun|BrokefromAngle Projectile)
    (LocalVariables|SetNormal (Variables|/Script/UnstableLiquidGunRuntime.FortLiquidGunProjectile|GetStopNormal Projectile))
    (LocalVariables|SetConsistent%20Random%20Vector (Math|Vector|SelectVector (Variables|/UnstableLiquidGun/Items/UnstableLiquidGun/B_Prj_Ranged_NitroglycerineGun.B_Prj_Ranged_NitroglycerineGun_C|GetRandomVectorOne Projectile) (Variables|/UnstableLiquidGun/Items/UnstableLiquidGun/B_Prj_Ranged_NitroglycerineGun.B_Prj_Ranged_NitroglycerineGun_C|GetRandomVectorTwo Projectile) IsBeginning))
    (LocalVariables|SetMultiplier (Math|Float|SelectFloat 1.0 -1.0 IsBeginning))
    (LocalVariables|SetSegmentDir (/ _returnvalue _directlength))
    (LocalVariables|SetSegmentLength (Math|Float|Clamp(Float) _directlength 0.0 (ScalableFloat|GetValueatLevel (Variables|Default|GetSF_Curling_MaxCurlOffset))))
    (return (* (+ (+ (Utilities|Operators|Multiply (Math|Vector|CrossProduct (* (LocalVariables|GetNormal) (LocalVariables|GetMultiplier)) (LocalVariables|GetSegmentDir)) (ScalableFloat|GetValueatLevel (Variables|Default|GetSF_Curling_TowardPerpendicularMultiplier)) (LocalVariables|GetSegmentLength)) (Utilities|Operators|Multiply _returnvalue (LocalVariables|GetMultiplier) (ScalableFloat|GetValueatLevel (Variables|Default|GetSF_Curling_TowardEndMultiplier)))) (Utilities|Operators|Multiply (Math|Vector|ProjectVectorontoPlane (LocalVariables|GetConsistent%20Random%20Vector) (LocalVariables|GetNormal)) (ScalableFloat|GetValueatLevel (Variables|Default|GetSF_Curling_TowardRandomMultiplier)) (LocalVariables|GetSegmentLength))) (Utilities|RemapFloat (- (Utilities|Time|GetGameTimeinSeconds) _timewhendetached_10_466cca0f4e3b02acaf08c39264b1976a) (Math|Vector2D|MakeVector2D 0.0 (ScalableFloat|GetValueatLevel (Variables|Default|GetSF_Curling_TimeToOffset))) "(X=0.000000,Y=1.000000)" true)))
    (else
      (Class|FortLiquidGunProjectile|SetDetachmentReason Projectile "ByDistance") ; Setting this so that projectile\nknows it's already been broken and\nthus doesn't later try to break from angle upon\ntouching the ground
      (return "0, 0, 0"))))

(fn GetCurvePoints
  (bind (_firstribbonid_6_d38b49c548211254b7303e91cc87c3d0 _secondribbonid_7_5dcd0da94598430c7a3ae389f35c563c _timewhendetached_10_466cca0f4e3b02acaf08c39264b1976a _segment_25_d2b36ed04b9565d0469a36b1ba847ab6 _segmentbeginning_16_66b4e98b4755ad97ca3f179663bb8a22 _segmentend_18_2208cfe047d7b1790a1862865f202f17) (Utilities|Struct|BreakLiquidGunDetachment (LocalVariables|GetCurrDetachment)))
  (bind (_firstribbonid_6_d38b49c548211254b7303e91cc87c3d0_2 _secondribbonid_7_5dcd0da94598430c7a3ae389f35c563c_2 _timewhendetached_10_466cca0f4e3b02acaf08c39264b1976a_2 _segment_25_d2b36ed04b9565d0469a36b1ba847ab6_2 _segmentbeginning_16_66b4e98b4755ad97ca3f179663bb8a22_2) (Utilities|Struct|BreakLiquidGunDetachment (LocalVariables|GetCurrDetachment)))
  (bind _maxsegmentlengthbeforebreak (Variables|LiquidRibbonManager|GetMaxSegmentLengthBeforeBreak))
  (bind _returnvalue (ScalableFloat|GetValueatLevel _maxsegmentlengthbeforebreak))
  (bind _pointspersegment (Variables|LiquidRibbonManager|GetPointsPerSegment))
  (bind _returnvalue_2 (Math|Conversions|ToFloat(Integer) _pointspersegment))
  (bind _firstvisible (CallFunction|BPGetFirstVisibleIndex))
  (bind _currentpointindex (LocalVariables|GetCurrentPointIndex))
  (bind _returnvalue_3 (> _currentpointindex _firstvisible))
  (bind _currentthickness (LocalVariables|GetCurrentThickness))
  (bind _closenesstobreakage (LocalVariables|GetClosenessToBreakage))
  (bind _output (Utilities|RemapFloat _closenesstobreakage "(X=0.540000,Y=0.950000)" (Math|Vector2D|MakeVector2D _currentthickness) true))
  (bind _returnvalue_4 (Math|Float|Round _output))
  (bind _returnvalue_5 (Math|Integer|SelectInt _returnvalue_4 0 _returnvalue_3))
  (bind _currentthickness_2 (LocalVariables|GetCurrentThickness))
  (bind _returnvalue_6 (Math|Float|Round _currentthickness_2))
  (bind _firstvisible_2 (CallFunction|BPGetFirstVisibleIndex))
  (bind _currentpointindex_2 (LocalVariables|GetCurrentPointIndex))
  (bind _returnvalue_7 (> _currentpointindex_2 _firstvisible_2))
  (bind _returnvalue_8 (Math|Integer|SelectInt _returnvalue_6 0 _returnvalue_7))
  (bind _pointspersegment_2 (Variables|LiquidRibbonManager|GetPointsPerSegment))
  (bind _returnvalue_9 (/ _pointspersegment_2 2))
  (bind _pointspersegment_3 (Variables|LiquidRibbonManager|GetPointsPerSegment))
  (bind _returnvalue_10 (/ _pointspersegment_3 2))
  (bind _returnvalue_11 (- _returnvalue_10 1))
  (LocalVariables|SetCurrentThickness (Variables|PointSize|GetMinPointSize))
  (bind _returnvalue_12 (CallFunction|CalculateSegments (Variables|Default|GetProjectilePositions) (Variables|LiquidRibbonManager|GetSplineCurveTension)))
  (for (_array_element _array_index) _returnvalue_12
    (LocalVariables|SetCurrentSegmentIndex _array_index)
    (if (Variables|Default|GetUNUSED%20LockedIn)
      (if (Utilities|Array|IsValidIndex (Variables|LiquidRibbonManager|GetSegmentLengths) (LocalVariables|GetCurrentSegmentIndex))
        _)
      (else
        (Utilities|Array|SetArrayElem (Variables|LiquidRibbonManager|GetSegmentLengths) (LocalVariables|GetCurrentSegmentIndex) (.DirectLength (Utilities|Struct|BreakLiquidRibbonSplineSegment _array_element)) true)))
    (if (and (< (.DirectLength (Utilities|Struct|BreakLiquidRibbonSplineSegment _array_element)) (ScalableFloat|GetValueatLevel (Variables|LiquidRibbonManager|GetMaxSegmentLengthBeforeBreak))) (not (Class|BPrjRangedNitroglycerineGun|BrokefromAngle (Utilities|GetClamped (Variables|Default|GetProjectiles) (LocalVariables|GetCurrentSegmentIndex)))))
      (for _index (range (Variables|LiquidRibbonManager|GetPointsPerSegment))
        (Math|Integer|IncrementInt (LocalVariables|GetCurrentPointIndex)
          (:%20%20
            (Utilities|Array|Add (LocalVariables|Get_Allpositions) (CallFunction|BPGetPointonSegment (/ (Math|Conversions|ToFloat(Integer) _index) (Variables|LiquidRibbonManager|GetPointsPerSegment)) _array_element))
            (Utilities|Array|Add (LocalVariables|GetRibbonVisibility) (and (CallFunction|BPPointIsVisible (LocalVariables|GetCurrentPointIndex) (CallFunction|BPGetFirstVisibleIndex)) (> (LocalVariables|GetCurrentPointIndex) (CallFunction|BPGetFirstVisibleIndex))))
            (Utilities|FlowControl|Sequence
              (:then_0
                (if (CallFunction|BPPointIsVisible (LocalVariables|GetCurrentPointIndex) (CallFunction|BPGetFirstVisibleIndex))
                  (bind _output_2 (CallFunction|BPOVCalculatePointSize (LocalVariables|GetCurrentSegmentIndex) (Variables|LiquidRibbonManager|GetSegmentLengths) (LocalVariables|GetCurrentPointIndex) (Variables|LiquidRibbonManager|GetPointsPerSegment) (Variables|PointSize|GetMinPointSize) (Variables|PointSize|GetMaxPointSize)))
                  (LocalVariables|SetCurrentThickness _output_2)))
              (:then_1
                (Utilities|Array|Add (LocalVariables|GetRibbonSizes) _returnvalue_8)
                (LocalVariables|SetLastAddedSize _returnvalue_8)
                (Utilities|Array|Add (LocalVariables|GetRibbonIDs) (LocalVariables|GetCurrentRibbonNumber))
                (bind _intensity (CallFunction|BPOVCalculatePointBlinkingIntensity (LocalVariables|GetLastAddedSize)))
                (Utilities|Array|Add (LocalVariables|Get_BlinkingIntensity) _intensity))))))
      (else
        ;[1 1 1] Detaching
        (bind _output_3 (CallFunction|BPGetorAddDetachment (LocalVariables|GetCurrentRibbonNumber) _array_element)) ; This WAS changed but I'm pretty sure it's now been changed back \nto what it originally was. Should be fine to leave as is but worth double\nchecking if something isn't working.
        (LocalVariables|SetCurrDetachment _output_3)
        (for _index_2 (range _returnvalue_10)
          (bind _output_get (LocalVariables|SetClosenessToBreakage (Utilities|RemapFloat (Math|Conversions|ToFloat(Integer) _index_2) (Math|Vector2D|MakeVector2D 0.0 (Math|Conversions|ToFloat(Integer) _returnvalue_11)) "(X=0.000000,Y=1.000000)" true)))
          (Math|Integer|IncrementInt (LocalVariables|GetCurrentPointIndex)
            (:%20%20
              (Utilities|Array|Add (LocalVariables|Get_Allpositions) (+ (Utilities|Operators|Multiply (Math|Curves|GetFloatValue (Variables|Default|GetCurlCurve) (Utilities|NormalizeFloat (- (Utilities|Time|GetGameTimeinSeconds) _timewhendetached_10_466cca0f4e3b02acaf08c39264b1976a_2) 0 (ScalableFloat|GetValueatLevel (Variables|Default|GetSF_TimeToBreak)) true)) (Math|Float|Clamp(Float) _output_get 0.0 0.75) (- (+ _segmentbeginning_16_66b4e98b4755ad97ca3f179663bb8a22_2 (CallFunction|CalculateOffset (LocalVariables|GetCurrDetachment) (Utilities|GetClamped (Variables|Default|GetProjectiles) (LocalVariables|GetCurrentSegmentIndex)) (Math|Float|Clamp(Float) _output_get 0.0 0.75) true)) (CallFunction|BPGetPointonSegment (/ (Math|Conversions|ToFloat(Integer) _index_2) (Math|Float|Truncate (Utilities|RemapFloat (.DirectLength (Utilities|Struct|BreakLiquidRibbonSplineSegment _array_element)) (Math|Vector2D|MakeVector2D _returnvalue 1e+04) (Math|Vector2D|MakeVector2D _returnvalue_2 1e+02) true))) _array_element))) (CallFunction|BPGetPointonSegment (/ (Math|Conversions|ToFloat(Integer) _index_2) (Math|Float|Truncate (Utilities|RemapFloat (.DirectLength (Utilities|Struct|BreakLiquidRibbonSplineSegment _array_element)) (Math|Vector2D|MakeVector2D _returnvalue 1e+04) (Math|Vector2D|MakeVector2D _returnvalue_2 1e+02) true))) _array_element)))
              (Utilities|Array|Add (LocalVariables|GetRibbonVisibility) (and (CallFunction|BPPointIsVisible (LocalVariables|GetCurrentPointIndex) (CallFunction|BPGetFirstVisibleIndex)) (> (LocalVariables|GetCurrentPointIndex) (CallFunction|BPGetFirstVisibleIndex))))
              ;[1 1 1] Making the size taper off then taper on when it's detached
              (Utilities|Array|Add (LocalVariables|GetRibbonSizes) _returnvalue_5)
              (LocalVariables|SetLastAddedSize _returnvalue_5))))
        (Math|Integer|IncrementInt (LocalVariables|GetCurrentRibbonNumber)
          (:%20%20
            (for _index_3 (range (- (Math|Float|Truncate (Utilities|RemapFloat (.DirectLength (Utilities|Struct|BreakLiquidRibbonSplineSegment _array_element)) (Math|Vector2D|MakeVector2D _returnvalue 1e+04) (Math|Vector2D|MakeVector2D _returnvalue_2 1e+02) true)) _returnvalue_9) (Math|Float|Truncate (Utilities|RemapFloat (.DirectLength (Utilities|Struct|BreakLiquidRibbonSplineSegment _array_element)) (Math|Vector2D|MakeVector2D _returnvalue 1e+04) (Math|Vector2D|MakeVector2D _returnvalue_2 1e+02) true)))
              (bind _output_get_2 (LocalVariables|SetClosenessToBreakage (Utilities|RemapFloat (Math|Conversions|ToFloat(Integer) _index_3) (Math|Vector2D|MakeVector2D (Math|Conversions|ToFloat(Integer) (- (Math|Float|Truncate (Utilities|RemapFloat (.DirectLength (Utilities|Struct|BreakLiquidRibbonSplineSegment _array_element)) (Math|Vector2D|MakeVector2D _returnvalue 1e+04) (Math|Vector2D|MakeVector2D _returnvalue_2 1e+02) true)) _returnvalue_9)) (Math|Conversions|ToFloat(Integer) (- (Math|Float|Truncate (Utilities|RemapFloat (.DirectLength (Utilities|Struct|BreakLiquidRibbonSplineSegment _array_element)) (Math|Vector2D|MakeVector2D _returnvalue 1e+04) (Math|Vector2D|MakeVector2D _returnvalue_2 1e+02) true)) 1))) "(X=1.000000,Y=0.000000)" true)))
              (Math|Integer|IncrementInt (LocalVariables|GetCurrentPointIndex)
                (:%20%20
                  (Utilities|Array|Add (LocalVariables|Get_Allpositions) (+ (Utilities|Operators|Multiply (Math|Curves|GetFloatValue (Variables|Default|GetCurlCurve) (Utilities|NormalizeFloat (- (Utilities|Time|GetGameTimeinSeconds) _timewhendetached_10_466cca0f4e3b02acaf08c39264b1976a) 0 (ScalableFloat|GetValueatLevel (Variables|Default|GetSF_TimeToBreak)) true)) (Math|Float|Clamp(Float) _output_get_2 0.0 0.75) (- (+ _segmentend_18_2208cfe047d7b1790a1862865f202f17 (CallFunction|CalculateOffset (LocalVariables|GetCurrDetachment) (Utilities|GetClamped (Variables|Default|GetProjectiles) (LocalVariables|GetCurrentSegmentIndex)) (Math|Float|Clamp(Float) _output_get_2 0.0 0.75))) (CallFunction|BPGetPointonSegment (/ (Math|Conversions|ToFloat(Integer) _index_3) (Math|Float|Truncate (Utilities|RemapFloat (.DirectLength (Utilities|Struct|BreakLiquidRibbonSplineSegment _array_element)) (Math|Vector2D|MakeVector2D _returnvalue 1e+04) (Math|Vector2D|MakeVector2D _returnvalue_2 1e+02) true))) _array_element))) (CallFunction|BPGetPointonSegment (/ (Math|Conversions|ToFloat(Integer) _index_3) (Math|Float|Truncate (Utilities|RemapFloat (.DirectLength (Utilities|Struct|BreakLiquidRibbonSplineSegment _array_element)) (Math|Vector2D|MakeVector2D _returnvalue 1e+04) (Math|Vector2D|MakeVector2D _returnvalue_2 1e+02) true))) _array_element)))))))))))
  (return (LocalVariables|GetRibbonIDs) (LocalVariables|GetRibbonSizes) (LocalVariables|Get_Allpositions) (LocalVariables|GetRibbonVisibility) (LocalVariables|Get_BlinkingIntensity)))

(fn GetFirstVisiblePointIndex
  (return (CallFunction|BPGetFirstVisibleIndex)))

(fn OLDCalculateSegment
  (params (p0 :type Vector) (p1 :type Vector) (p2 :type Vector) (p3 :type Vector) (01SqrtDist :type Float%20%28double-precision%29) (12SqrtDist :type Float%20%28double-precision%29) (23SqrtDist :type Float%20%28double-precision%29))
  ;[1 1 1] This is all the math equation for creating a spline
  (bind _output_get (LocalVariables|Sett0 0.0))
  (bind _output_get_2 (LocalVariables|Sett1 (+ _output_get 01SqrtDist)))
  (bind _output_get_3 (LocalVariables|Sett2 (+ _output_get_2 12SqrtDist)))
  (LocalVariables|Sett3 (+ _output_get_3 23SqrtDist))
  (LocalVariables|Setm1 (* (- (LocalVariables|Gett2) (LocalVariables|Gett1)) (+ (- (/ (- (Variables|Getp1) (Variables|Getp0)) (Math|Float|Max(Float) (- (LocalVariables|Gett1) (LocalVariables|Gett0)) 0.01)) (/ (- (Variables|Getp2) (Variables|Getp0)) (Math|Float|Max(Float) (- (LocalVariables|Gett2) (LocalVariables|Gett0)) 0.01))) (/ (- (Variables|Getp2) (Variables|Getp1)) (Math|Float|Max(Float) (- (LocalVariables|Gett2) (LocalVariables|Gett1)) 0.01)))))
  (LocalVariables|Setm2 (* (- (LocalVariables|Gett2) (LocalVariables|Gett1)) (+ (- (/ (- (Variables|Getp2) (Variables|Getp1)) (Math|Float|Max(Float) (- (LocalVariables|Gett2) (LocalVariables|Gett1)) 0.01)) (/ (- (Variables|Getp3) (Variables|Getp1)) (Math|Float|Max(Float) (- (LocalVariables|Gett3) (LocalVariables|Gett1)) 0.01))) (/ (- (Variables|Getp3) (Variables|Getp2)) (Math|Float|Max(Float) (- (LocalVariables|Gett3) (LocalVariables|Gett2)) 0.01)))))
  (return (Utilities|Struct|MakeLiquidGunCurveSegment (Utilities|Operators|Add (* 2.0 (- (Variables|Getp1) (Variables|Getp2))) (LocalVariables|Getm1) (LocalVariables|Getm2)) (Utilities|Operators|Subtract (* -3.0 (- (Variables|Getp1) (Variables|Getp2))) (LocalVariables|Getm1) (LocalVariables|Getm1) (LocalVariables|Getm2)) (LocalVariables|Getm1) (Variables|Getp1) (Math|Float|Square 12SqrtDist)) 0 0 0))

(fn OLDCalculateSegments
  (params (Points :type Array%20of%20Vectors))
  (bind _returnvalue (Utilities|Array|Length Points))
  (LocalVariables|SetmaxIndex (- _returnvalue 1))
  (Utilities|Array|Resize (LocalVariables|GetSqrtDistances) _returnvalue)
  (for (_array_element _array_index) Points
    (Utilities|Array|SetArrayElem (LocalVariables|GetSqrtDistances) _array_index (Math|Float|Sqrt (Math|Vector|Distance(Vector) _array_element (Utilities|Array|Get(acopy) Points (Math|Integer|Min(Integer) (+ _array_index 1) (LocalVariables|GetmaxIndex)))))))
  (for (_array_element_2 _array_index_2) Points
    (bind _newparam (CallFunction|OLDCalculateSegment (Utilities|Array|Get(acopy) (Variables|GetPoints) (Math|Integer|Max(Integer) (- _array_index_2 1))) _array_element_2 (Utilities|Array|Get(acopy) (Variables|GetPoints) (Math|Integer|Min(Integer) (+ _array_index_2 1) (LocalVariables|GetmaxIndex))) (Utilities|Array|Get(acopy) (Variables|GetPoints) (Math|Integer|Min(Integer) (+ _array_index_2 2) (LocalVariables|GetmaxIndex))) (Utilities|Array|Get(acopy) (LocalVariables|GetSqrtDistances) (Math|Integer|Max(Integer) (- _array_index_2 1))) (Utilities|Array|Get(acopy) (LocalVariables|GetSqrtDistances) _array_index_2) (Utilities|Array|Get(acopy) (LocalVariables|GetSqrtDistances) (Math|Integer|Min(Integer) (+ _array_index_2 1) (LocalVariables|GetmaxIndex))))) ; Uses 4 adjacent points to create segments
    (Utilities|Array|Add (LocalVariables|Getsegments) _newparam))
  (return (LocalVariables|Getsegments)))

(fn UpdateNiagaraPositions
  (params (RunOnServer :type Boolean))
  (bind _weapon (Variables|Default|GetWeapon))
  (bind _returnvalue (Effects|GetMuzzleLocation _weapon -1))
  (bind _firstvisible (CallFunction|BPGetFirstVisibleIndex))
  (bind _projectiles (Variables|Default|GetProjectiles))
  (bind _output (Utilities|Array|Get(acopy) _projectiles))
  (bind _returnvalue_2 (+ _firstvisible 2))
  (bind _returnvalue_3 (+ _firstvisible 1))
  (bind _projectiles_2 (Variables|Default|GetProjectiles))
  (bind _projectilepositions (Variables|Default|GetProjectilePositions))
  (bind _ribbonfx (Variables|Default|GetRibbonFX))
  (bind _sizes (Variables|Default|GetSizes))
  (if (Variables|Default|GetUseNative)
    ;[0.1491 1 0.4606] Nativized Version
    ; Note: some of the BP version is unused (eg Previous/CurrentLocation and Stopped) and is thus not included in the native version
    (Variables|Default|GetWeapon
      (:then
        (bind (_outallpositions _outribbonids _outribbonvisibilities _outflashintensities _outribbonsizes) (CallFunction|CalculateSplinePointsandSetNiagaraData _projectiles_2 (Math|Boolean|ANDBoolean (not (Networking|HasAuthority)) (Utilities|IsValid (Variables|Default|GetWeapon) :!class /Script/Engine.KismetSystemLibrary) (Variables|/UnstableLiquidGun/Items/UnstableLiquidGun/B_Weapon_LiquidGun.B_Weapon_LiquidGun_C|GetMidStream (Variables|Default|GetWeapon)) (== (Variables|/UnstableLiquidGun/Items/UnstableLiquidGun/B_Weapon_LiquidGun.B_Weapon_LiquidGun_C|GetCurrentManager (Variables|Default|GetWeapon)) self)) _ribbonfx))
        (Variables|Default|SetAllPositions _outallpositions)
        (Variables|Default|SetBlinkingIntensityArray _outflashintensities)
        (Variables|Default|SetSizes _outribbonsizes))
      (:else
        (bind (_outallpositions_2 _outribbonids_2 _outribbonvisibilities_2 _outflashintensities_2 _outribbonsizes_2) (CallFunction|CalculateSplinePointsandSetNiagaraData _projectiles_2 false _ribbonfx))
        (Variables|Default|SetAllPositions _outallpositions_2)
        (Variables|Default|SetBlinkingIntensityArray _outflashintensities_2)
        (Variables|Default|SetSizes _outribbonsizes_2)))
    (elif (Utilities|Array|IsValidIndex _projectiles)
      (Utilities|IsValid _output
        (:Is%20Valid
          (LocalVariables|SetPreviousLocation (Transformation|GetActorLocation _output))
          (if (not (and (Variables|Default|GetStopped) (not RunOnServer)))
            (Utilities|Array|Clear (Variables|Default|GetProjectilePositions))
            (for (_array_element _array_index) (Variables|Default|GetProjectiles)
              (Utilities|IsValid _array_element
                (:Is%20Valid
                  (LocalVariables|SetIndex _array_index)
                  (if (Variables|/Script/Engine.Actor|GetbHidden _array_element)
                    _
                    (elif (and (Variables|/UnstableLiquidGun/Items/UnstableLiquidGun/B_Prj_Ranged_NitroglycerineGun.B_Prj_Ranged_NitroglycerineGun_C|GetAttachedToPawn _array_element) (Utilities|IsValid (Variables|/Script/UnstableLiquidGunRuntime.FortLiquidGunProjectile|GetAttachComponent _array_element) :!class /Script/Engine.KismetSystemLibrary))
                      (LocalVariables|SetCurrentLocation (+ (Transformation|GetWorldLocation (Variables|/Script/UnstableLiquidGunRuntime.FortLiquidGunProjectile|GetAttachComponent _array_element)) (Variables|/Script/UnstableLiquidGunRuntime.FortLiquidGunProjectile|GetAttachOffset _array_element)))
                      (Utilities|Array|Add (Variables|Default|GetProjectilePositions) (LocalVariables|GetCurrentLocation))
                      (else
                        (LocalVariables|SetCurrentLocation (Transformation|GetActorLocation _array_element))))))))
            (if (and (Variables|/UnstableLiquidGun/Items/UnstableLiquidGun/B_Weapon_LiquidGun.B_Weapon_LiquidGun_C|GetMidStream (Variables|Default|GetWeapon)) (== (Variables|/UnstableLiquidGun/Items/UnstableLiquidGun/B_Weapon_LiquidGun.B_Weapon_LiquidGun_C|GetCurrentManager (Variables|Default|GetWeapon)) self))
              (Networking|SwitchHasAuthority
                (:Authority
                  _)
                (:Remote
                  (Utilities|Array|Add _projectilepositions _returnvalue)
                  (Utilities|Array|Add _projectilepositions _returnvalue)))
              (bind (_ribbonids1 _ribbonsizes1 _allpositions _ribbonvisibilityarray _blinkingintensity) (Default|GetCurvePoints))
              (Variables|Default|SetAllPositions _allpositions)
              (Variables|Default|SetBlinkingIntensityArray _blinkingintensity)
              (bind _ribbonfx_2 (Variables|Default|GetRibbonFX)
                (:then
                  (Niagara|NiagaraSetVectorArray _ribbonfx_2 "PosArray" _allpositions)
                  (Niagara|NiagaraSetInt32Array _ribbonfx_2 "RibbonIDArray" _ribbonids1)
                  (Niagara|NiagaraSetBoolArray _ribbonfx_2 "RibbonVisibilityArray" _ribbonvisibilityarray)
                  (Niagara|NiagaraSetFloatArray _ribbonfx_2 "fIsFlashingArray" _blinkingintensity)
                  (bind _output_get (Variables|Default|SetSizes _ribbonsizes1))
                  (for _index (range 1 16)
                    (if (Utilities|Array|IsValidIndex (Variables|Default|GetSizes) (- (Utilities|Array|Length _output_get) 15))
                      (if (Utilities|Array|IsValidIndex (Variables|Default|GetSizes) (- (Utilities|Array|Length _output_get) _index))
                        (Utilities|Array|SetArrayElem (Variables|Default|GetSizes) (- (Utilities|Array|Length _output_get) _index) (Math|Float|Round (Utilities|RemapFloat (Math|Conversions|ToFloat(Integer) _index) "(X=15.000000,Y=0.000000)" (Math|Vector2D|MakeVector2D (Math|Conversions|ToFloat(Integer) (Utilities|Array|Get(acopy) (Variables|Default|GetSizes) (- (Utilities|Array|Length _output_get) 15))) 6e+01) true))))))
                  (if (Utilities|Array|IsValidIndex _sizes)
                    (Utilities|Array|SetArrayElem _sizes)
                    (if (Utilities|Array|IsValidIndex _sizes 1)
                      (Utilities|Array|SetArrayElem _sizes 1)
                      (if (Utilities|Array|IsValidIndex _sizes _returnvalue_3)
                        (Utilities|Array|SetArrayElem _sizes _returnvalue_3)
                        (if (Utilities|Array|IsValidIndex _sizes _returnvalue_2)
                          (Utilities|Array|SetArrayElem _sizes _returnvalue_2)
                          (bind _ribbonfx_3 (Variables|Default|GetRibbonFX)
                            (:then
                              (Niagara|NiagaraSetInt32Array _ribbonfx_3 "RibbonSizes" (Variables|Default|GetSizes)))
                            (:else)))))))
                (:else))
              (else
                _))))))))

(fn ConstructionScript)
