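; Weapon Blueprint for the Unstable Liquid Gun, which fires a projectile stream and removes itself from inventory when out of ammo.
; Each projectile joins a linked list and a per-stream ribbon manager, with syncing for explosions across managers and optional spiral spread.
; Trigger and burst events track stream state and drive hand-managed looping audio, montages and canister liquid Niagara effects.
; Mid-stream cleanup can block firing through a GameplayEffect, removed when firing ends.
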(event FortWeapon|EventOnSpawnProjectile
  (params (SpawnProjectile :type Fort%20Projectile%20Base%20Object%20Reference))
  (bind _lastprj (Variables|Default|GetLastPrj))
  (Utilities|FlowControl|Sequence
    (:then_0
      ;[1 1 1] Spiral Spread (Hotfixed off by default)
      (if (ScalableFloat|GetValueasBool (Variables|Default|GetSF_SpiralSpread_Enabled))
        (Projectile|SetVelocity (Variables|/Script/FortniteGame.FortProjectileBase|GetProjectileMovementComponent SpawnProjectile) (Utilities|Operators|Add (Variables|/Script/Engine.MovementComponent|GetVelocity (Variables|/Script/FortniteGame.FortProjectileBase|GetProjectileMovementComponent SpawnProjectile)) (* (Math|Vector|MakeVector 0.0 0.0 (* (Utilities|TimeSinceStreamStarted) (ScalableFloat|GetValueatLevel (Variables|Default|GetSF_SpiralSpread_RadiusIncreaseOverTime)))) (Math|Trig|Cos(Radians) (* (Utilities|Time|GetGameTimeinSeconds) (ScalableFloat|GetValueatLevel (Variables|Default|GetSF_SpiralSpread_RotationSpeed))))) (Utilities|Operators|Multiply (Math|Vector|CrossProduct (Math|Vector|Normalize (Variables|/Script/Engine.MovementComponent|GetVelocity (Variables|/Script/FortniteGame.FortProjectileBase|GetProjectileMovementComponent SpawnProjectile))) "0, 0,1.000000") (* (Utilities|TimeSinceStreamStarted) (ScalableFloat|GetValueatLevel (Variables|Default|GetSF_SpiralSpread_RadiusIncreaseOverTime))) (Math|Trig|Sin(Radians) (* (Utilities|Time|GetGameTimeinSeconds) 12.0)))))))
    (:then_1
      ;[1 1 1] Set up linked list of projectiles (each one references the next one and the previous one)
      (bind _asb_prj_ranged_nitroglycerine_gun (Utilities|Casting|CastToB_Prj_Ranged_NitroglycerineGun SpawnProjectile))
      (Class|BPrjRangedNitroglycerineGun|SetPreviousPrj _asb_prj_ranged_nitroglycerine_gun _lastprj :!class /UnstableLiquidGun/Items/UnstableLiquidGun/B_Prj_Ranged_NitroglycerineGun.B_Prj_Ranged_NitroglycerineGun_C)
      (Class|BPrjRangedNitroglycerineGun|SetNextPrj _lastprj _asb_prj_ranged_nitroglycerine_gun :!class /UnstableLiquidGun/Items/UnstableLiquidGun/B_Prj_Ranged_NitroglycerineGun.B_Prj_Ranged_NitroglycerineGun_C)
      (Variables|Default|SetLastPrj _asb_prj_ranged_nitroglycerine_gun))
    (:then_2
      (if (and (not (Variables|Default|GetMidStream)) (Variables|Default|GetFiring))
        ;[1 1 1] Spawn manager
        (bind _returnvalue (Game|SpawnActorfromClass "/UnstableLiquidGun/Items/UnstableLiquidGun/BP_UnstableLiquidGun_RibbonManager.BP_UnstableLiquidGun_RibbonManager_C" (Math|Transform|MakeTransform (Transformation|GetActorLocation)) "AlwaysSpawn" "OverrideRootScale" 0 (Variables|Actor|GetInstigator)))
        (bind _output_get (Variables|Default|SetCurrentManager _returnvalue))
        (Variables|/UnstableLiquidGun/Items/UnstableLiquidGun/BP_LiquidGun_RibbonManager.BP_LiquidGun_RibbonManager_C|SetWeapon self _output_get)
        (Variables|Default|SetMidStream true)
        (Variables|Default|SetStreamStartTime (Utilities|Time|GetGameTimeinSeconds))
        (Variables|HandlingInfiniteAmmo|SetMaxShotsLeftPerManager (Math|Random|RandomIntegerinRange (* (FortWeapon|GetBulletsPerClip self) 2) (* (FortWeapon|GetBulletsPerClip self) 3)))
        (elif (<= (FortWeapon|GetMagazineAmmoCount) 0)
          (Variables|Default|SetMidStream false)
          (else
            (bind _result (Math|Integer|DecrementInt (Variables|HandlingInfiniteAmmo|GetMaxShotsLeftPerManager))
              (:%20%20
                (if (not (>= _result 0))
                  (Variables|HandlingInfiniteAmmo|SetPreviousManager (Variables|Default|GetCurrentManager))))))))
      ;[0.2286 0.4698 0.1527 0.5] If we made a new manager mid-stream, tell it to wait for the previous one's explosions to end. This keeps the visuals synced so the fuse-style explosions play seamlessly between managers and niagara systems.
      (bind _previousmanager (Variables|HandlingInfiniteAmmo|GetPreviousManager)
        (:then
          (Class|BPUnstableLiquidGunRibbonManager|WaittoExplodeonDispatch (Variables|Default|GetCurrentManager) _previousmanager)
          (Variables|HandlingInfiniteAmmo|SetPreviousManager 0) ; Clear the value
        )
        (:else
          _))
      ;[1 1 1] Add new projectile to manager
      (Class|BPLiquidGunRibbonManager|AddProjectile (Variables|Default|GetCurrentManager) (Variables|Default|GetLastPrj)))
    (:then_3
      (if (<= (FortWeapon|GetRemainingAmmo) 0)
        (Inventory|RemoveItemFromPlayerByGuid (Fortnite|GetFortPlayerControllerfromActor (Game|GetInstigator)) (FortWeapon|GetInventoryGUID) 1 true)))))

(event EventBeginPlay
  (Default|BindEventtoOnStartFiring (EventDispatchers|CreateEvent :!event Custom|OnStartFiring_Event)) ; Called on all Clients
  (Default|BindEventtoOnPersistentFireStopped (delegate Custom|OnPersistentFireStopped_Event)) ; Called on all Clients
  (EventDispatchers|BindEventtoOnReleaseTrigger (delegate Custom|OnReleaseTrigger_)) ; Called on Local Client
  (EventDispatchers|BindEventtoOnPressTrigger (EventDispatchers|CreateEvent :!event Custom|FirePressed)) ; Called on local client
  ;[1 1 1] Looping audio for firing. Implementing looping sounds in defaults seems to repeat the fire sound every projectile spawned. This is a workaround to get smooth looping
  ; Also handles custom looping fire animations
  (EventDispatchers|BindEventtoOnLocalAmmoChanged (delegate Custom|LocalAmmoChanged))
  (Default|BindEventtoOnStartFiring (delegate Custom|StartFiring)))

(event Custom|HandleManagerCleanUpMidStream
  ;[0.7365 0.2907 0.2222 0.5] Handle manager cleanup mid-stream
  (Networking|CanExecuteCosmeticEvents
    (:True
      (Variables|HandlingInfiniteAmmo|SetbTemporarilyDisableFiring true)
      (CallFunction|HandleManagerCleanUpMidStreamOwningClienttoServer)
      (CallFunction|EndFiringFX true))) ; For whatever reason, this only ever seems to come from the Client.\nBut we're doing this check to be safe because the reason is unclear.
)

(event Custom|HandleOnFire
  (bind _returnvalue (Game|GetInstigator))
  (bind _returnvalue_2 (Pawn|IsLocallyControlled _returnvalue))
  (bind _sfx_fireloop_1p (Variables|Audio|GetSFX_FireLoop_1P))
  (bind _sfx_fireloop_3p (Variables|Audio|GetSFX_FireLoop_3P))
  (bind _returnvalue_3 (select _returnvalue_2 _sfx_fireloop_1p _sfx_fireloop_3p))
  ;[1 1 1] Looping audio for firing. Implementing looping sounds in defaults seems to repeat the fire sound every projectile spawned. This is a workaround to get smooth looping
  ; Also handles custom looping fire animations
  (Networking|CanExecuteCosmeticEvents
    (:True
      (if (not (Variables|HandlingInfiniteAmmo|GetbTemporarilyDisableFiring))
        (Utilities|FlowControl|Sequence
          (:then_0
            (Variables|Default|GetFireLoop
              (:then)
              (:else
                (bind _returnvalue_4 (Audio|SpawnSoundAttached _returnvalue_3 (Variables|Visual|GetWeaponMesh)))
                (Variables|Default|SetFireLoop _returnvalue_4)
                (Variables|Default|SetbPlayedStopAudio false)
                (SoundIndicator|StartSoundIndicatorLooping (Variables|SoundIndicator|GetSoundIndicatorComponent) (Transformation|GetActorLocation) _returnvalue_3 1.0 0))))
          (:then_1
            (CallFunction|PlayStartFiringMontages)))))))

(event Custom|OnReleaseTrigger_
  (params (Weapon :type Fort%20Weapon%20Object%20Reference))
  (CallFunction|FireReleased))

(event Custom|HandleManagerCleanUpMidStream_OwningClientToServer
  ;[0.7365 0.2907 0.2222 0.5] Handle manager cleanup mid-stream
  (Networking|SwitchHasAuthority
    (:Authority
      (Fortnite|ApplyGameplayEffectsToActor (Game|GetInstigator) (Utilities|Array|MakeArray (Variables|HandlingInfiniteAmmo|GetGE_UnstableLiquidGun_DisableFire)) (Transformation|GetActorLocation (Game|GetInstigator)) false 0 (Ability|GetFortAbilitySystemComponent (Game|GetInstigator))) ; Disable primary fire via GE
    )))

(event Custom|StartFiring
  ;[1 1 1] Looping audio for firing. Implementing looping sounds in defaults seems to repeat the fire sound every projectile spawned. This is a workaround to get smooth looping
  ; Also handles custom looping fire animations
  (if (> (FortWeapon|GetMagazineAmmoCount) 0)
    (CallFunction|HandleonFire)))

(event Custom|OnPersistentFireStopped_Event
  (CallFunction|BurstEnded))

(event Custom|EndFiringFX
  (params (bIsEquipped :type Boolean))
  ;[1 1 1] Looping audio for firing. Implementing looping sounds in defaults seems to repeat the fire sound every projectile spawned. This is a workaround to get smooth looping
  ; Also handles custom looping fire animations
  (Utilities|FlowControl|Sequence
    (:then_0
      (Audio|Components|Audio|FadeOut (Variables|Default|GetFireLoop) 0.5 0.0 "Linear" :!class /Script/Engine.AudioComponent)
      (Variables|Default|SetFireLoop 0)
      (if (and bIsEquipped (not (Variables|Default|GetbPlayedStopAudio)))
        (Audio|SpawnSoundAttached (select (Pawn|IsLocallyControlled (Game|GetInstigator)) (Variables|Audio|GetSFX_FireStop_1P) (Variables|Audio|GetSFX_FireStop_3P)) (Variables|Visual|GetWeaponMesh))
        (Variables|Default|SetbPlayedStopAudio true)
        (SoundIndicator|StopSoundIndicatorLooping (Variables|SoundIndicator|GetSoundIndicatorComponent))))
    (:then_1
      (CallFunction|PlayEndFiringMontages))))

(event Custom|LocalAmmoChanged
  (params (LocalCount :type Integer) (LocalRemaining :type Integer))
  (bind _playliquidfillonrolad (Variables|B_Weapon_LiquidGun|GetPlayLiquidFillOnRolad))
  ;[1 1 1] Looping audio for firing. Implementing looping sounds in defaults seems to repeat the fire sound every projectile spawned. This is a workaround to get smooth looping
  ; Also handles custom looping fire animations
  (bind _output_get (Variables|Default|SetLocalAmmoCount LocalCount))
  (CallFunction|PrintStringInternal (Utilities|String|Append "LocalAmmoCount: [" (Utilities|String|ToString(Integer) _output_get) "]") "(R=1.000000,G=0.000000,B=0.000000,A=1.000000)" 4.0)
  (Utilities|FlowControl|Sequence
    (:then_0
      (if (<= _output_get 0)
        (CallFunction|HandleonEndFire true)))
    (:then_1
      (Networking|CanExecuteCosmeticEvents
        (:True
          (if (and (Components|Timeline|IsPlaying _playliquidfillonrolad) (Variables|Default|GetFiring))
            (Components|Timeline|SetNewTime _playliquidfillonrolad 1.8))
          (Effects|Components|ParticleSystem|SetFloatParameter (Variables|Default|GetNS_UnstableLiquidGun_CanisterLiquid) "PistonProgress" (Utilities|RemapFloat (Math|Conversions|ToFloat(Integer) _output_get) "(X=0.000000,Y=20.000000)" "(X=0.000000,Y=1.000000)"))
          (Variables|Default|SetPistonProgressCache (Utilities|RemapFloat (Math|Conversions|ToFloat(Integer) _output_get) "(X=0.000000,Y=20.000000)" "(X=0.000000,Y=1.000000)")))))))

(event Custom|HandleOnEndFire
  (params (bIsEquipped :type Boolean))
  ;[1 1 1] Looping audio for firing. Implementing looping sounds in defaults seems to repeat the fire sound every projectile spawned. This is a workaround to get smooth looping
  ; Also handles custom looping fire animations
  (Utilities|FlowControl|Sequence
    (:then_0
      (Networking|CanExecuteCosmeticEvents
        (:True
          (Variables|HandlingInfiniteAmmo|SetbTemporarilyDisableFiring false)
          (CallFunction|EndFiringFX bIsEquipped))))
    (:then_1
      (Networking|SwitchHasAuthority
        (:Authority
          (GameplayEffects|RemoveActiveGameplayEffectbySourceEffect (Ability|GetFortAbilitySystemComponent (Game|GetInstigator)) (Variables|HandlingInfiniteAmmo|GetGE_UnstableLiquidGun_DisableFire) (Ability|GetFortAbilitySystemComponent (Game|GetInstigator)) -1) ; Only remove the GE when the player releases the fire input or unequips this weapon.
        )))))

(event FortWeapon|EventOnEquippedAndAttached
  ;[1 1 1] Looping audio for firing. Implementing looping sounds in defaults seems to repeat the fire sound every projectile spawned. This is a workaround to get smooth looping
  ; Also handles custom looping fire animations
  (Effects|Components|ParticleSystem|SetBoolParameter (Variables|Default|GetNS_UnstableLiquidGun_CanisterLiquid) "WeaponVisible" true)
  (Variables|Default|SetLocalAmmoCount (FortWeapon|GetLocalMagazineAmmoCount)))

(event FortWeapon|EventOnUnEquippedAndDetached
  ;[1 1 1] Looping audio for firing. Implementing looping sounds in defaults seems to repeat the fire sound every projectile spawned. This is a workaround to get smooth looping
  ; Also handles custom looping fire animations
  ;;[1 1 1] Stop firing loop if weapon is switched or otherwise hidden (tac sprint, etc)
  (Effects|Components|ParticleSystem|SetBoolParameter (Variables|Default|GetNS_UnstableLiquidGun_CanisterLiquid) "WeaponVisible" false)
  (CallFunction|HandleonEndFire false)
  (Variables|Default|SetMidStream false))

(event EventOnWeaponVisibilityChanged
  (params (bVisible :type Boolean) (bSetForLocalControllerOnly :type Boolean))
  ;[1 1 1] Looping audio for firing. Implementing looping sounds in defaults seems to repeat the fire sound every projectile spawned. This is a workaround to get smooth looping
  ; Also handles custom looping fire animations
  ;;[1 1 1] Stop firing loop if weapon is switched or otherwise hidden (tac sprint, etc)
  (Effects|Components|ParticleSystem|SetBoolParameter (Variables|Default|GetNS_UnstableLiquidGun_CanisterLiquid) "WeaponVisible" bVisible)
  (if (not bVisible)
    (CallFunction|HandleonEndFire false)
    (Variables|Default|SetMidStream false)))

(event EventOnPlayReloadFX
  (params (ReloadStage :type EFortReloadFXState%20Enum :default "ReloadStart"))
  ;[1 1 1] PlayReloadingfX
  (|CallParentFunction ReloadStage)
  (switch Utilities|FlowControl|Switch|SwitchonEFortReloadFXState ReloadStage
    (:ReloadStart
      (if (FortWeapon|IsReloading)
        (Utilities|FlowControl|Sequence
          (:then_0
            (Utilities|FlowControl|Delay 0.2)
            (Effects|Components|ParticleSystem|SetFloatParameter (Variables|Default|GetNS_UnstableLiquidGun_CanisterLiquid_Small) "FillAmount" 1.0))
          (:then_1
            (Utilities|FlowControl|Delay 0.45)
            (Components|Timeline|PlayfromStart (Variables|B_Weapon_LiquidGun|GetPlayLiquidFillOnRolad))
            (Effects|Components|ParticleSystem|SetFloatParameter (Variables|Default|GetNS_UnstableLiquidGun_CanisterLiquid) "FillAmount" 0.0)))))))

(event Custom|FirePressed
  (params (Weapon :type Fort%20Weapon%20Object%20Reference))
  (Variables|Default|SetFireButtonPressed true)
  (CallFunction|PrintStringInternal "Fire button pressed." "(R=0.000000,G=1.000000,B=1.000000,A=1.000000)" 5.0))

(event Custom|EndedFiring
  (Variables|Default|SetFiring false)
  (Variables|Default|SetMidStream false)
  (Variables|Default|SetMidBurstLocal false)
  (CallFunction|PrintStringInternal "Ended Firing")
  (CallFunction|HandleonEndFire true)
  (bind _output_get (Variables|Default|SetStreamStartTime -1.0)))

(event Custom|FireReleased
  (Variables|Default|SetFireButtonPressed false)
  (Development|PrintString "Fire button released" (Variables|Default|GetbDebug) (Variables|Default|GetbDebug))
  (if (and (not (Variables|Default|GetMidBurstLocal)) (Variables|Default|GetFiring))
    (CallFunction|PrintStringInternal "Fire released, still firing, but not mid burst. Ending Burst again")
    (CallFunction|BurstEnded)))

(event Custom|PlayStartFiringMontages
  (if (not (Variables|Default|GetbFiringMontagesActive))
    ;[0.7042 1 0.6753] Play the pawn Start (+ loop) firing montage
    (bind _playerstartfiringmontage (Variables|Default|GetPlayerStartFiringMontage)
      (:then
        (bind _asfort_player_pawn (Utilities|Casting|CastToFortPlayerPawn (Game|GetInstigator))
          (:then
            (Animation|Montage|PlayMontage (Variables|/Script/Engine.Character|GetMesh _asfort_player_pawn) _playerstartfiringmontage))
          (:CastFailed
            _))
        ;[1 1 1] Temp - Start looping weapon montage (Will eventually be setup with start+loop and end similar to the player)
        (bind _tempweaponloopmontage (Variables|Default|GetTempWeaponLoopMontage)
          (:then
            (Animation|Montage|PlayMontage (Variables|Visual|GetWeaponMesh) _tempweaponloopmontage))
          (:else
            _))
        (Variables|Default|SetbFiringMontagesActive true))
      (:else
        _))))

(event Custom|OnStartFiring_Event
  (Variables|Default|SetMidBurstLocal true) ; OnStartFiring is called each burst
  (CallFunction|OnStartFiringServer)
  (Variables|Default|SetFiring true)
  (CallFunction|PrintStringInternal "Start Fire"))

(event Custom|PlayEndFiringMontages
  (if (Variables|Default|GetbFiringMontagesActive)
    ;[1 0.6281 0.442] Play the pawn End firing montage
    (bind _playerendfiringmontage (Variables|Default|GetPlayerEndFiringMontage)
      (:then
        (bind _asfort_player_pawn (Utilities|Casting|CastToFortPlayerPawn (Game|GetInstigator))
          (:then
            (Animation|Montage|PlayMontage (Variables|/Script/Engine.Character|GetMesh _asfort_player_pawn) _playerendfiringmontage))
          (:CastFailed
            _))
        ;[1 1 1] Temp - Stop looping weapon montage
        (bind _tempweaponloopmontage (Variables|Default|GetTempWeaponLoopMontage)
          (:then
            (bind _weaponmesh (Variables|Visual|GetWeaponMesh)
              (:then
                (Utilities|IsValid (Components|SkeletalMesh|GetAnimInstance _weaponmesh)
                  (:Is%20Valid
                    (Animation|Montage|MontageStop (Components|SkeletalMesh|GetAnimInstance _weaponmesh) 0.0 _tempweaponloopmontage))
                  (:Is%20Not%20Valid
                    _))
                (Variables|Default|SetbFiringMontagesActive false))
              (:else
                _)))
          (:else
            _)))
      (:else
        _))))

(event Custom|OnStartFiring_Server
  (Variables|Default|SetFiring true)
  (CallFunction|PrintStringInternal "Start Fire"))

(event Custom|BurstEnded
  (bind _returnvalue (Game|GetInstigator))
  (bind _returnvalue_2 (Fortnite|GetFortPlayerControllerfromActor _returnvalue))
  (Variables|Default|SetMidBurstLocal false)
  (CallFunction|PrintStringInternal "Burst Ended")
  (if (not (<= (FortWeapon|GetLocalMagazineAmmoCount) 0))
    (Utilities|IsValid _returnvalue
      (:Is%20Valid
        (Utilities|IsValid _returnvalue_2
          (:Is%20Valid
            (if (and (Pawn|IsLocalPlayerController _returnvalue_2) (not (Variables|Default|GetFireButtonPressed)))
              _))))
      (:Is%20Not%20Valid
        _)))
  (CallFunction|EndedFiring(Server))
  (CallFunction|EndedFiring))

(event Custom|EndedFiring%20%28Server%29
  (CallFunction|EndedFiring))

(event Custom|ResetForcedInput
  ;[1 0 0] Deprecated
  (Input|ResetForcedInputRotationSpeed (Fortnite|GetFortPlayerControllerfromActor (Game|GetInstigator)))
  (Variables|Default|SetVelocityIncrease 0.0))

(fn PrintStringInternal
  (params (InString :type String) (TextColor :type Linear%20Color%20Structure :default "(R=0.000000,G=1.000000,B=1.000000,A=1.000000)") (Duration :type Float%20%28single-precision%29 :default 2.0))
  (Development|PrintString (Utilities|String|Append "B_Weapon_LiquidGun " InString ".") (Variables|Default|GetbDebug) (Variables|Default|GetbDebug) TextColor Duration))

(fn ConstructionScript
  (|CallParentFunction))
