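# Material function for Unstable Liquid Gun FX that builds a banded four-color wrap pattern from value noise over scaled coordinates.
# Repeated noise, optionally quantized into steps, drives a four-way color lerp, though every color input is currently constant white.
# A sphere mask around the player position, with radius set by ColorsBlendFactor, blends color pairs unless a scalar bypasses it.
# It also outputs average intensity and two blend masks, one always using world position.
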MaterialFunction /UnstableLiquidGun/Items/UnstableLiquidGun/FX/MF_Liquid_WrapPattern
graph {nodes [
  n1 = Constant3Vector() {constant=<rgba 1 1 1 0>}
  n2 = CollectionParameter() {collection=&FortniteMaterialParameters, parameterName=PlayerPosition}
  n3 = FunctionInput(Preview=ComponentMask(None=@n2) {b=true, g=true, r=true}) {bUsePreviewValueAsDefault=true, inputName=BlendOrigin}
  n4 = WorldPosition()
  n5 = Constant() {r=1}
  n6 = MaterialFunctionCall!(
   1=Constant()
   2=Constant() {r=4000}
   3=Constant() {r=9000}
   4=Constant() {r=40000}
   A=Saturate(None=FunctionInput(Preview=@n5) {bUsePreviewValueAsDefault=true, inputName=ColorsBlendFactor, inputType=FunctionInput_Scalar})
  ) {materialFunction=/Engine/Functions/Engine_MaterialFunctions02/Lerp_Multiple_Float}
  n7 = Constant() {r=0.5}
  n8 = FunctionInput(Preview=ScalarParameter() {parameterName=Don't Use Wrap World Blend}) {bUsePreviewValueAsDefault=true, inputName=Don't Use Wrap World Blend, inputType=FunctionInput_Scalar}
  n9 = LinearInterpolate(
   A=SphereMask(A=@n3, B=FunctionInput(Preview=@n4.XYZ) {bUsePreviewValueAsDefault=true, inputName=WorldPosForBlend}, Radius=@n6.Lerp 4 Inputs, Hardness=@n7)
   Alpha=@n8
  )
  n10 = LinearInterpolate(A=ComponentMask(None=@n1) {b=true, g=true, r=true}, B=ComponentMask(None=@n1) {b=true, g=true, r=true}, Alpha=@n9)
  n11 = LinearInterpolate(A=ComponentMask(None=@n1) {b=true, g=true, r=true}, B=ComponentMask(None=@n1) {b=true, g=true, r=true}, Alpha=@n9)
  n12 = LinearInterpolate(A=ComponentMask(None=@n1) {b=true, g=true, r=true}, B=ComponentMask(None=@n1) {b=true, g=true, r=true}, Alpha=@n9)
  n13 = ComponentMask(None=@n1) {b=true, g=true, r=true}
  n14 = ComponentMask(None=@n1) {b=true, g=true, r=true}
  n15 = AppendVector(A=TextureCoordinate(), B=Constant())
  n16 = ScalarParameter() {defaultValue=0.2, parameterName=PatternScale}
  n17 = Noise!(
   World Position=Divide(
    A=FunctionInput(Preview=@n15) {bUsePreviewValueAsDefault=true, inputName=Coordinates}
    B=FunctionInput(Preview=@n16) {inputName=PatternScale, inputType=FunctionInput_Scalar}
   )
   FilterWidth
  ) {bTurbulence=false, levels=1, noiseFunction=NOISEFUNCTION_ValueALU, outputMin=0}
  n18 = ScalarParameter() {defaultValue=1, group=Dance Lines, parameterName=ColorRepetition}
  n19 = Frac(None=Multiply(A=ComponentMask(None=@n17) {g=true}, B=FunctionInput(Preview=@n18) {inputName=ColorRepetition, inputType=FunctionInput_Scalar}))
  n20 = FunctionInput(Preview=ScalarParameter() {defaultValue=100, parameterName=GradientQuantisationLevels}) {bUsePreviewValueAsDefault=true, inputName=GradientSoftness, inputType=FunctionInput_Scalar}
  n21 = Divide(A=Floor(None=Multiply(A=@n19, B=@n20)), B=@n20)
  n22 = FunctionInput(Preview=ScalarParameter() {defaultValue=1, parameterName=GradientQuantisationBlending}) {bUsePreviewValueAsDefault=true, inputName=GradientQuantisationBlending, inputType=FunctionInput_Scalar}
  FunctionOutput(None=MaterialFunctionCall(1=@n10, 2=@n11, 3=@n12, 4=LinearInterpolate(A=@n13, B=@n14, Alpha=@n9), A=LinearInterpolate(A=@n19, B=@n21, Alpha=@n22)) {materialFunction=&Lerp_Multiple_Float3}.Lerp 4 Inputs)
  n23 = Divide(A=Add(A=Add(A=@n10, B=@n11), B=@n12)) {constB=3}
  FunctionOutput(None=Divide(A=MaterialFunctionCall(f3=@n23) {materialFunction=&AddComponents}.f3) {constB=3}) {outputName=AverageIntensity}
  FunctionOutput(None=@n9) {outputName=BlendMask}
  FunctionOutput(None=LinearInterpolate(A=SphereMask(A=@n3, B=@n4.XYZ, Radius=@n6.Lerp 4 Inputs, Hardness=@n7), Alpha=@n8)) {outputName=BlendMask_WP}
 ]}
