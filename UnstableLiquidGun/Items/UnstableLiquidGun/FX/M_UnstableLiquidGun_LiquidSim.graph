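# Masked two-sided material for the Unstable Liquid Gun that fakes sloshing liquid inside a container, built for Niagara mesh particles.
# Dynamic parameters supply plane orientation, fill, size, speed and time, with a static switch for some non-Niagara fallbacks.
# A rotated position gradient plus cosine and noise waves forms the cut waterline and meniscus.
# Two-sided sign gives back faces surface normals and front faces turbulence, while vertex color masks animated sphere offsets.
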Material /UnstableLiquidGun/Items/UnstableLiquidGun/FX/M_UnstableLiquidGun_LiquidSim
graph {
 properties {bUsedWithNiagaraMeshParticles=true, blendMode=BLEND_Masked, pixelDepthOffsetMode=PDOM_Legacy, twoSided=true, usedShadingModels=Default Lit}
 nodes [
  Add(A=Multiply(A=Time(), B=ScalarParameter() {defaultValue=-0.5, parameterName=Param_3}) {constB=50}, B=NamedRerouteUsage())
  Constant3Vector() {constant=<rgba 1 1 1 0>}
  n1 = WorldPosition()
  n2 = NamedRerouteDeclaration(None=StaticSwitchParameter(True=ParticlePositionWS(), False=ObjectPositionWS()) {defaultValue=true, parameterName=UsedWithNiagara}) {name=Position, nodeColor=<rgba 0.4297 0 1 1>}
  n3 = Multiply(A=NamedRerouteUsage(), B=Constant3Vector() {constant=<rgba -1 -1 1 0>})
  n4 = Divide(
   A=MaterialFunctionCall!(
    VectorToRotate=Subtract(A=@n1.XYZ, B=@n2)
    LookAtVector=Normalize(VectorInput=@n3)
    RestingVector=Constant3Vector() {constant=<rgba 0 0 1 0>}
   ) {materialFunction=/Engine/Functions/Engine_MaterialFunctions01/Vectors/RotateVector}.Result
   B=NamedRerouteUsage()
  ) {constB=1}
  n5 = ComponentMask(None=@n4) {b=true}
  n6 = LinearInterpolate(Alpha=NamedRerouteUsage()) {constA=0.5, constB=-0.5}
  n7 = Multiply(A=Normalize(VectorInput=@n4), B=NamedRerouteUsage())
  n8 = Multiply(A=Power(Base=MaterialFunctionCall(Vector 2=@n7) {materialFunction=/Engine/Functions/Engine_MaterialFunctions02/Utility/VectorLength}.V2 Length)) {constB=0.5}
  n9 = NamedRerouteUsage()
  n10 = ScalarParameter() {defaultValue=4, parameterName=TileMult}
  n11 = LinearInterpolate(A=Constant2Vector() {g=0.8, r=0.7}, B=Constant2Vector() {g=0.1, r=0.9}, Alpha=NamedRerouteUsage())
  n12 = Multiply(
   A=Multiply(A=Multiply(A=@n10, B=@n11), B=NamedRerouteUsage())
   B=LinearInterpolate(Alpha=ScalarParameter() {defaultValue=0.6, parameterName=Viscosity}) {constA=1, constB=0.5}
  )
  n13 = NamedRerouteUsage()
  n14 = Subtract(A=TextureSample(UVs=Panner(Coordinate=@n12, Time=@n13) {speedY=-1.5}) {samplerType=SAMPLERTYPE_LinearColor, texture=&HighContrastNoise}.RGB) {constB=0.3}
  n15 = NamedRerouteUsage()
  n16 = LinearInterpolate(A=Multiply(A=Power(Base=@n15), B=NamedRerouteUsage()), Alpha=NamedRerouteUsage())
  n17 = MaterialFunctionCall(Color=Multiply(A=Multiply(A=@n14, B=@n16)) {constB=0.75}) {materialFunction=&SplitComponents}
  n18 = Add(A=Multiply(A=Cosine(None=@n8), B=Multiply(A=@n9) {constB=-0.5}) {constB=0.15}, B=Add(A=@n17.R, B=@n17.R)) {constB=0}
  n19 = Add(A=Subtract(A=Multiply(B=@n5) {constA=-1}, B=@n6), B=@n18) {constB=0}
  n20 = SmoothStep(Value=Multiply(A=NamedRerouteUsage(), B=NamedRerouteDeclaration(None=@n19) {name=Water Line Gradient, nodeColor=<rgba 1 0 0.1172 1>}) {constA=1}) {constMax=0.001}
  LinearInterpolate(A=@n20, B=OneMinus(None=@n20), Alpha=NamedRerouteUsage()) {constA=-1}
  n21 = VectorParameter() {parameterName=MainColorC, defaultValue=<rgba 1 0.2606 0.06 0>}
  n22 = VectorParameter() {parameterName=MainColorD, defaultValue=<rgba 0.5052 0.1765 0.1159 0>}
  n23 = NamedRerouteUsage()
  n24 = ComponentMask(None=@n23) {g=true, r=true}
  n25 = ScalarParameter() {defaultValue=1, parameterName=Param_2}
  n26 = TextureSample(
   UVs=Panner(Coordinate=Multiply(A=@n24, B=@n25) {constB=0.5}) {speedX=0.02, speedY=0.02}
   Tex=TextureObject() {samplerType=SAMPLERTYPE_LinearGrayscale, texture=/Game/Global/Textures/Default/Flowmaps/T_StormSwirlMask}
  )
  n27 = ScalarParameter() {defaultValue=0.5, parameterName=Param_4}
  n28 = Multiply(
   A=LinearInterpolate(A=Multiply(A=@n21.RGB, B=@n22.RGB), B=@n21.RGB, Alpha=Power(Base=@n26.RGB, Exp=@n27) {constExponent=0.5})
   B=ScalarParameter() {defaultValue=0.5, parameterName=Param_6}
  )
  MaterialFunctionCall(Emissive=@n28) {materialFunction=&MF_FX_EmissiveAdjustments}
  MaterialFunctionCall(
   TextureObject=TextureObject() {samplerType=SAMPLERTYPE_Masks, texture=/Game/Weapons/WeaponSkins/Wraps/Textures/Mash/Mash_Noise}
   TextureSize=Multiply(A=ScalarParameter() {defaultValue=20, parameterName=Param_7}, B=@n25)
   WorldPosition=MaterialFunctionCall() {materialFunction=&LocalPosition}.Local Position
  ) {materialFunction=/Game/Weapons/WeaponSkins/Wraps/MaterialFunctions/WorldAlignedTexture}
  Multiply(A=@n25, B=Constant2Vector() {g=1, r=3})
  n29 = DynamicParameter() {defaultValue=<rgba 0 0 1 0>, paramNames [Param1,Param2,Param3,Speed]}
  NamedRerouteDeclaration(None=Multiply(A=@n29.Speed, B=ScalarParameter() {defaultValue=2, parameterName=SpeedMultWaves})) {name=Speed, nodeColor=<rgba 0 0.7891 1 1>}
  n30 = DynamicParameter() {parameterIndex=1, defaultValue=<rgba -1 0 0 0>, paramNames [UpsideDown,Time,WaveAdd,Param4]}
  NamedRerouteDeclaration(None=StaticSwitchParameter(True=@n30.WaveAdd, False=ScalarParameter() {parameterName=WaveAdd}) {defaultValue=true, parameterName=UsedWithNiagara}) {name=WaveAdd, nodeColor=<rgba 1 0 0.5156 1>}
  n31 = VectorParameter() {parameterName=MainColorA, defaultValue=<rgba 0.6406 0.4245 0.3989 1>}
  n32 = VectorParameter() {parameterName=MainColorB, defaultValue=<rgba 1 0.3987 0.2686 0>}
  n33 = ScalarParameter() {defaultValue=2, parameterName=Param_5}
  NamedRerouteDeclaration(None=LinearInterpolate(A=Multiply(A=@n31.RGB, B=@n32.RGB) {constB=0.4}, B=@n31.RGB, Alpha=Power(Base=@n26.RGB, Exp=@n33) {constExponent=0.4})) {name=BaseColor, nodeColor=<rgba 1 0.9844 0 1>}
  n34 = NamedRerouteUsage()
  n35 = Multiply(A=NamedRerouteUsage(), B=VectorParameter() {parameterName=ScreenSpaceUVOffsetMultiply, defaultValue=<rgba -10 -10 -4 0>}.RGB) {constB=10}
  n36 = NamedRerouteUsage()
  NamedRerouteDeclaration(None=MaterialFunctionCall!(Position=Subtract(A=@n34, B=@n35), UVRadialScale=Divide(B=@n36)) {materialFunction=/Game/Weapons/WeaponSkins/MaterialFunction/ObjectStabilizedScreenSpaceUVs}.StabilizedUVs) {name=ScreenSpaceUV, nodeColor=<rgba 1 0.9375 0 1>}
  NamedRerouteDeclaration(None=Multiply(A=Add(A=@n30.UpsideDown)) {constB=0.5}) {name=WaterUpsideDown, nodeColor=<rgba 0 0.1797 1 1>}
  n37 = Constant() {r=0.013}
  n38 = LinearInterpolate(A=@n37, B=Multiply(A=@n37) {constB=-1}, Alpha=NamedRerouteUsage()) {constA=0.015}
  n39 = Subtract(A=@n19, B=@n38) {constB=0}
  n40 = OneMinus(None=Saturate(None=Divide(A=@n39) {constB=0.015}))
  n41 = Saturate(None=Divide(A=NamedRerouteUsage(), B=@n38) {constB=0.02})
  n42 = LinearInterpolate(A=OneMinus(None=@n41), B=@n41, Alpha=NamedRerouteUsage())
  n43 = NamedRerouteUsage()
  NamedRerouteDeclaration(None=Multiply(A=Subtract(A=@n40, B=@n42) {constA=0}, B=LinearInterpolate(Alpha=@n43) {constA=0.333, constB=3}) {constB=0.333}) {name=Miniscus, nodeColor=<rgba 0 1 0.5313 1>}
  NamedRerouteDeclaration(None=Subtract(A=WorldPosition().XYZ, B=ActorPositionWS())) {name=Position, nodeColor=<rgba 1 0.037 0 1>}
  n44 = ComponentMask(None=NamedRerouteUsage()) {g=true, r=true}
  n45 = Constant2Vector() {g=0.01, r=0.06}
  n46 = Multiply(A=ScalarParameter() {defaultValue=4, parameterName=TileMult}, B=ScalarParameter() {defaultValue=0.7, parameterName=Bubbles_Tile})
  n47 = Multiply(A=Add(A=Multiply(A=@n44, B=@n45) {constB=0.05}, B=NamedRerouteUsage()), B=Multiply(A=@n46, B=Constant2Vector() {g=1.5, r=2}) {constA=2})
  n48 = NamedRerouteUsage()
  n49 = Multiply(A=@n46, B=Constant2Vector() {g=2})
  n50 = TextureSample(UVs=BumpOffset(Coordinate=Panner(Coordinate=@n47, Time=@n48, Speed=@n49) {speedY=3}, Height=NamedRerouteUsage(), HeightRatioInput=Constant() {r=0.66}) {heightRatio=0.6}) {
   samplerSource=SSM_Wrap_WorldGroupSettings
   samplerType=SAMPLERTYPE_Normal
   texture=/Game/Characters/Player/Female/Medium/Bodies/F_MED_Bistro_Astronaut/FX/T_GaussianNoise_Normal
  }
  NamedRerouteDeclaration(None=@n50.RGB) {name=Bubbles Normal, nodeColor=<rgba 0.3359 1 0 1>}
  n51 = TextureSampleParameter2D() {parameterName=Normal, samplerType=SAMPLERTYPE_Normal, texture=&FlatNormalBlue}
  n52 = NamedRerouteUsage()
  n53 = LinearInterpolate(A=NamedRerouteUsage(), B=@n51.RGB, Alpha=@n52) {desc=Surface normal}
  n54 = LinearInterpolate(A=@n51.RGB, B=NamedRerouteUsage(), Alpha=NamedRerouteUsage())
  n55 = NamedRerouteUsage()
  n56 = Constant2Vector() {g=0.2, r=0.2}
  n57 = Panner(Coordinate=Multiply(A=ScalarParameter() {defaultValue=4, parameterName=TileMult}, B=Multiply(A=@n55, B=@n56)), Time=NamedRerouteUsage()) {speedY=-0.7}
  n58 = TextureSampleParameter2D(UVs=@n57) {
   parameterName=Normal_Side
   samplerType=SAMPLERTYPE_Normal
   texture=&T_GaussianNoise_Normal
  }
  n59 = NamedRerouteDeclaration(None=@n58.RGB) {name=Water Turbulence, nodeColor=<rgba 1 0 0.8203 1>}
  n60 = Power(Base=Saturate(None=NamedRerouteUsage())) {constExponent=0.7}
  n61 = LinearInterpolate(
   A=ShadingPathSwitch(Default=LinearInterpolate(A=@n54, B=@n59, Alpha=@n60), Forward=@n51.RGB, Mobile=@n51.RGB) {desc=Simplify for switch/mobile}
   B=Multiply(A=NamedRerouteUsage()) {constB=-1}
   Alpha=@n52
  ) {desc=Side normal}
  n62 = Saturate(None=TwoSidedSign())
  NamedRerouteDeclaration(None=Normalize(VectorInput=LinearInterpolate(A=@n53, B=@n61, Alpha=@n62))) {name=Pixel Normals, nodeColor=<rgba 0 0.8125 1 1>}
  NamedRerouteDeclaration(None=DotProduct(A=VertexNormalWS(), B=CameraVectorWS())) {name=PixelBumpOffset, nodeColor=<rgba 0.0312 0 1 1>}
  n63 = VertexNormalWS()
  n64 = Multiply(A=OneMinus(None=Fresnel(Normal=@n63) {baseReflectFraction=0, exponent=2}), B=NamedRerouteUsage())
  n65 = Saturate(None=Multiply(A=NamedRerouteUsage()) {constB=3})
  NamedRerouteDeclaration(None=Multiply(A=LinearInterpolate(A=@n64, Alpha=@n65) {constB=0}, B=ScalarParameter() {defaultValue=1, parameterName=Bubbles Multiplier})) {name=Bubbles Mask, nodeColor=<rgba 0.7813 0 1 1>}
  NamedRerouteDeclaration(None=@n28) {name=Emissive, nodeColor=<rgba 0 0.8125 1 1>}
  n66 = Transform(None=NamedRerouteUsage()) {transformSourceType=TRANSFORMSOURCE_World, transformType=TRANSFORM_Tangent}
  n67 = NamedRerouteUsage()
  n68 = NamedRerouteUsage()
  n69 = Multiply(A=NamedRerouteUsage(), B=ScalarParameter() {defaultValue=4, parameterName=Water Surface Normal Tile})
  n70 = Multiply(
   A=LinearInterpolate(Alpha=NamedRerouteUsage()) {constA=1.2, constB=0.8}
   B=LinearInterpolate(Alpha=ScalarParameter() {defaultValue=0.6, parameterName=Viscosity}) {constA=1, constB=1.5}
  )
  n71 = MaterialFunctionCall(
   Plane Normal Axis=NamedRerouteUsage()
   Plane X Axis=Constant3Vector() {constant=<rgba 1 0 0 0>}
   Plane Center=Subtract(A=NamedRerouteUsage(), B=Add(A=@n67, B=@n68))
   UV Scale=ComponentMask(None=Multiply(A=@n69, B=@n70)) {g=true, r=true}
  ) {materialFunction=&VirtualPlaneCoordinates}
  n72 = NamedRerouteUsage()
  n73 = TextureSampleParameter2D(UVs=Panner(Coordinate=@n71.UVs, Time=@n72) {speedY=0.4}) {parameterName=Normal_Top, samplerType=SAMPLERTYPE_Normal, texture=/FortWater/Textures/02_WaterFoam_Veroni_base_N}
  n74 = Constant() {r=0.25}
  n75 = TextureSampleParameter2D(UVs=Panner(
    Coordinate=MaterialFunctionCall!(Angle=@n74, UV/2DVector=@n71.UVs) {materialFunction=/Game/Characters/Player/Male/Medium/Bodies/M_MED_Faux/Materials/MF_2DRotate}.Result
    Time=@n72
   ) {speedX=0.4}) {parameterName=Normal_Top, samplerType=SAMPLERTYPE_Normal, texture=&02_WaterFoam_Veroni_base_N}
  n76 = Multiply(
   A=MaterialFunctionCall(X=@n75.G, Y=@n75.G, Z=@n75.G) {materialFunction=&MakeFloat3}.Result
   B=Constant3Vector() {constant=<rgba 1 -1 1 0>}
  )
  n77 = Saturate(None=Add(A=NamedRerouteUsage(), B=NamedRerouteUsage()))
  n78 = Multiply(
   A=Multiply(
    A=MaterialFunctionCall(BaseNormal=@n73.RGB, AdditionalNormal=@n76) {materialFunction=&BlendAngleCorrectedNormals}.Result
    B=Constant3Vector() {desc=Renormal, constant=<rgba -1 1 -1 0>}
   ) {constB=-1}
   B=LinearInterpolate(A=Constant3Vector() {constant=<rgba 0 0 1 0>}, Alpha=Power(Base=@n77) {constExponent=0.6})
  )
  NamedRerouteDeclaration(None=Normalize(VectorInput=MaterialFunctionCall(BaseNormal=@n66, AdditionalNormal=@n78) {materialFunction=&BlendAngleCorrectedNormals}.Result)) {name=Water Plane Normal, nodeColor=<rgba 0 0.6016 1 1>}
  NamedRerouteDeclaration(None=Normalize(VectorInput=MaterialFunctionCall(X=@n29.Param1, Y=@n29.Param1, Z=@n29.Param1) {materialFunction=&MakeFloat3}.Result)) {name=WaterPlaneZ, nodeColor=<rgba 1 0 0.1875 1>}
  n79 = Multiply(A=Time()) {constB=0.25}
  n80 = ScalarParameter() {defaultValue=0.6, parameterName=Viscosity}
  NamedRerouteDeclaration(None=Multiply(
    A=StaticSwitchParameter(True=@n30.Time, False=@n79) {defaultValue=true, parameterName=UsedWithNiagara}
    B=LinearInterpolate(Alpha=@n80) {constA=1, constB=0.15}
   )) {name=Viscosity, nodeColor=<rgba 0 1 0.5078 1>}
  NamedRerouteDeclaration(None=@n18) {name=Height, nodeColor=<rgba 0 1 0.8828 1>}
  n81 = MaterialFunctionCall(f3=VertexColor()) {materialFunction=&AddComponents}
  NamedRerouteDeclaration(None=StaticSwitchParameter(True=Saturate(None=@n81.f3), False=Constant()) {defaultValue=true, parameterName=UsedWithNiagara}) {name=SpheresMask, nodeColor=<rgba 0.7578 0 1 1>}
  NamedRerouteDeclaration(None=Multiply(A=ScalarParameter() {defaultValue=0.5, parameterName=SpeedMultBubbles}, B=@n29.Speed)) {name=Bubbles Intensity, nodeColor=<rgba 0 0.7891 1 1>}
  n82 = DynamicParameter() {parameterIndex=3, defaultValue=<rgba 30 30 50 0.6>}
  NamedRerouteDeclaration(None=@n82.A) {name=Fill, nodeColor=<rgba 0 1 0.0859 1>}
  NamedRerouteDeclaration(None=StaticSwitchParameter(
    True=MaterialFunctionCall(X=@n82.R, Y=@n82.R, Z=@n82.R) {materialFunction=&MakeFloat3}.Result
    False=VectorParameter() {parameterName=ContainerSize}.RGB
   ) {defaultValue=true, parameterName=UsedWithNiagara}) {name=Dimensions, nodeColor=<rgba 0 0.0994 1 1>}
  n83 = NamedRerouteUsage()
  NamedRerouteDeclaration(None=Multiply(A=NamedRerouteUsage(), B=Multiply(A=@n83, B=@n6))) {name=Center, nodeColor=<rgba 1 0 0.0937 1>}
  NamedRerouteUsage()
  NamedRerouteUsage()
  NamedRerouteUsage()
  NamedRerouteUsage()
  ReflectionVectorWS(CustomWorldNormal=@n23) {bNormalizeCustomWorldNormal=true}
  ScalarParameter() {defaultValue=0.1, parameterName=Roughness}
  ScalarParameter() {defaultValue=0.5, parameterName=Metallic}
  ScalarParameter() {defaultValue=1, parameterName=Specular}
  n84 = Multiply(A=NamedRerouteUsage()) {constB=0.5}
  n85 = NamedRerouteUsage()
  n86 = Multiply(A=Add(A=NamedRerouteUsage(), B=NamedRerouteUsage()), B=Multiply(A=ScalarParameter() {defaultValue=1.5}, B=Add(A=@n84, B=@n85)))
  n87 = VertexColor()
  n88 = Add(A=NamedRerouteUsage(), B=Time())
  n89 = Multiply(
   A=RotateAboutAxis(
    NormalizedRotationAxis=Add(A=@n87) {constB=0.15}
    RotationAngle=Multiply(A=@n88) {constB=0.3}
    PivotPoint=Constant()
    Position=NamedRerouteUsage()
   )
   B=ScalarParameter() {defaultValue=2.5, parameterName=Param_1}
  )
  StaticSwitchParameter(True=LinearInterpolate(B=Add(A=@n86, B=@n89), Alpha=NamedRerouteUsage()), False=Constant()) {defaultValue=true, parameterName=UsedWithNiagara}
  TextureCoordinate()
  TextureSample() {
   samplerSource=SSM_Wrap_WorldGroupSettings
   samplerType=SAMPLERTYPE_Normal
   texture=/Game/Effects/Fort_Effects/Textures/Tile/T_CellNoise_Tile_N
  }
  TextureSample(UVs=@n57) {samplerType=SAMPLERTYPE_Normal, texture=&02_WaterFoam_Veroni_base_N}
 ]
}
