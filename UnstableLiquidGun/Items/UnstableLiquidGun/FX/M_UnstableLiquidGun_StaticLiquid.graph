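# Masked, two-sided lit material for the Unstable Liquid Gun FX, usable on Niagara mesh particles, that fakes liquid filling a container.
# A water-line gradient from container size, fill and panning noise height feeds a SmoothStep cutoff.
# Back faces pose as the liquid surface, using virtual-plane foam normals and, apparently, stabilized screen-space UVs.
# A panning swirl mask blends base and emissive colors, while Viscosity scales wave speed and tiling.
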Material /UnstableLiquidGun/Items/UnstableLiquidGun/FX/M_UnstableLiquidGun_StaticLiquid
graph {
 properties {bUsedWithNiagaraMeshParticles=true, blendMode=BLEND_Masked, pixelDepthOffsetMode=PDOM_Legacy, twoSided=true, usedShadingModels=Default Lit}
 nodes [
  n1 = VectorParameter() {parameterName=MainColorA, defaultValue=<rgba 0.6406 0.4245 0.3989 1>}
  n2 = VectorParameter() {parameterName=MainColorB, defaultValue=<rgba 1 0.3987 0.2686 0>}
  n3 = Multiply(A=ComponentMask(None=NamedRerouteUsage()) {g=true, r=true}, B=Constant() {r=0.2}) {constB=0.5}
  n4 = Multiply(A=TextureCoordinate(), B=Constant2Vector() {g=0.1, r=1})
  n5 = Saturate(None=TwoSidedSign())
  n6 = TextureSample(
   UVs=Panner(Coordinate=LinearInterpolate(A=@n3, B=@n4, Alpha=@n5)) {speedX=0.02, speedY=0.02}
   Tex=TextureObject() {samplerType=SAMPLERTYPE_LinearGrayscale, texture=/Game/Global/Textures/Default/Flowmaps/T_StormSwirlMask}
  )
  n7 = ScalarParameter() {defaultValue=2, parameterName=Param_5}
  NamedRerouteDeclaration(None=LinearInterpolate(A=Multiply(A=@n1.RGB, B=@n2.RGB) {constB=0.4}, B=@n1.RGB, Alpha=Power(Base=@n6.RGB, Exp=@n7) {constExponent=0.4})) {name=BaseColor, nodeColor=<rgba 1 0.9844 0 1>}
  n8 = NamedRerouteUsage()
  n9 = Multiply(A=NamedRerouteUsage(), B=VectorParameter() {parameterName=ScreenSpaceUVOffsetMultiply, defaultValue=<rgba -5 -5 -4 0>}.RGB) {constB=10}
  n10 = NamedRerouteUsage()
  NamedRerouteDeclaration(None=MaterialFunctionCall!(Position=Subtract(A=@n8, B=@n9), UVRadialScale=Divide(B=@n10)) {materialFunction=/Game/Weapons/WeaponSkins/MaterialFunction/ObjectStabilizedScreenSpaceUVs}.StabilizedUVs) {name=ScreenSpaceUV, nodeColor=<rgba 1 0.9375 0 1>}
  n11 = NamedRerouteUsage()
  n12 = TextureSampleParameter2D() {parameterName=Normal, samplerType=SAMPLERTYPE_Normal, texture=&FlatNormalBlue}
  n13 = ShadingPathSwitch(Default=@n12.RGB, Forward=@n12.RGB, Mobile=@n12.RGB) {desc=Simplify for switch/mobile}
  n14 = Saturate(None=TwoSidedSign())
  NamedRerouteDeclaration(None=Normalize(VectorInput=LinearInterpolate(A=@n11, B=@n13, Alpha=@n14))) {name=Pixel Normals, nodeColor=<rgba 0 0.8125 1 1>}
  n15 = VectorParameter() {parameterName=MainColorC, defaultValue=<rgba 1 0.2606 0.06 0>}
  n16 = Multiply(A=@n15.RGB, B=VectorParameter() {parameterName=MainColorD, defaultValue=<rgba 0.5052 0.1765 0.1159 0>}.RGB)
  n17 = Power(Base=@n6.RGB, Exp=ScalarParameter() {defaultValue=0.5, parameterName=Param_4}) {constExponent=0.5}
  NamedRerouteDeclaration(None=Multiply(A=LinearInterpolate(A=@n16, B=@n15.RGB, Alpha=@n17), B=ScalarParameter() {defaultValue=0.5, parameterName=Param_6})) {name=Emissive, nodeColor=<rgba 0 0.8125 1 1>}
  n18 = Transform(None=NamedRerouteUsage()) {transformSourceType=TRANSFORMSOURCE_World, transformType=TRANSFORM_Tangent}
  n19 = NamedRerouteUsage()
  n20 = NamedRerouteUsage()
  n21 = Multiply(A=NamedRerouteUsage(), B=ScalarParameter() {defaultValue=20, parameterName=Water Surface Normal Tile})
  n22 = Multiply(B=LinearInterpolate(Alpha=ScalarParameter() {defaultValue=0.8, parameterName=Viscosity}) {constA=1, constB=1.5}) {constA=1.2}
  n23 = MaterialFunctionCall(
   Plane Normal Axis=NamedRerouteUsage()
   Plane X Axis=Constant3Vector() {constant=<rgba 1 0 0 0>}
   Plane Center=Subtract(A=NamedRerouteUsage(), B=Add(A=@n19, B=@n20))
   UV Scale=ComponentMask(None=Multiply(A=@n21, B=@n22)) {g=true, r=true}
  ) {materialFunction=&VirtualPlaneCoordinates}
  n24 = NamedRerouteUsage()
  n25 = TextureSampleParameter2D(UVs=Panner(Coordinate=@n23.UVs, Time=@n24) {speedY=0.4}) {parameterName=Normal_Top, samplerType=SAMPLERTYPE_Normal, texture=/FortWater/Textures/02_WaterFoam_Veroni_base_N}
  n26 = Constant() {r=0.25}
  n27 = TextureSampleParameter2D(UVs=Panner(
    Coordinate=MaterialFunctionCall!(Angle=@n26, UV/2DVector=@n23.UVs) {materialFunction=/Game/Characters/Player/Male/Medium/Bodies/M_MED_Faux/Materials/MF_2DRotate}.Result
    Time=@n24
   ) {speedX=0.4}) {parameterName=Normal_Top, samplerType=SAMPLERTYPE_Normal, texture=&02_WaterFoam_Veroni_base_N}
  n28 = Multiply(
   A=MaterialFunctionCall(X=@n27.G, Y=@n27.G, Z=@n27.G) {materialFunction=&MakeFloat3}.Result
   B=Constant3Vector() {constant=<rgba 1 -1 1 0>}
  )
  n29 = Multiply(
   A=Multiply(
    A=MaterialFunctionCall(BaseNormal=@n25.RGB, AdditionalNormal=@n28) {materialFunction=&BlendAngleCorrectedNormals}.Result
    B=Constant3Vector() {desc=Renormal, constant=<rgba -1 1 -1 0>}
   ) {constB=-1}
   B=LinearInterpolate(A=Constant3Vector() {constant=<rgba 0 0 1 0>}, Alpha=Constant() {r=1})
  )
  NamedRerouteDeclaration(None=Normalize(VectorInput=MaterialFunctionCall(BaseNormal=@n18, AdditionalNormal=@n29) {materialFunction=&BlendAngleCorrectedNormals}.Result)) {name=Water Plane Normal, nodeColor=<rgba 0 0.6016 1 1>}
  NamedRerouteDeclaration(None=Normalize(VectorInput=Constant3Vector() {constant=<rgba 0 0 1 0>})) {name=WaterPlaneZ, nodeColor=<rgba 1 0 0.1875 1>}
  n30 = Time()
  n31 = ScalarParameter() {defaultValue=0.8, parameterName=Viscosity}
  NamedRerouteDeclaration(None=Multiply(A=Multiply(A=@n30) {constB=0.25}, B=LinearInterpolate(Alpha=@n31) {constA=1, constB=0.15})) {name=Viscosity, nodeColor=<rgba 0 1 0.5078 1>}
  n32 = Multiply(A=Multiply(A=ScalarParameter() {defaultValue=2, parameterName=TileMult}, B=Constant2Vector() {g=0.8, r=0.7}), B=NamedRerouteUsage())
  n33 = LinearInterpolate(Alpha=ScalarParameter() {defaultValue=0.8, parameterName=Viscosity}) {constA=1, constB=0.5}
  n34 = TextureSample(UVs=Panner(Coordinate=Multiply(A=@n32, B=@n33), Time=NamedRerouteUsage()) {speedY=-1.5}) {samplerType=SAMPLERTYPE_LinearColor, texture=&HighContrastNoise}
  n35 = MaterialFunctionCall(Color=Multiply(A=Subtract(A=@n34.RGB) {constB=0.3}) {constB=0.75}) {materialFunction=&SplitComponents}
  n36 = Add(A=@n35.R, B=@n35.R)
  NamedRerouteDeclaration(None=@n36) {name=Height, nodeColor=<rgba 0 1 0.8828 1>}
  n37 = VectorParameter() {parameterName=ContainerSize, defaultValue=<rgba 1.5 1.5 1.5 5.5>}
  NamedRerouteDeclaration(None=@n37.A) {name=Fill, nodeColor=<rgba 0 1 0.0859 1>}
  NamedRerouteDeclaration(None=@n37.RGB) {name=Dimensions, nodeColor=<rgba 0 0.0994 1 1>}
  n38 = NamedRerouteUsage()
  n39 = LinearInterpolate(Alpha=NamedRerouteUsage()) {constA=0.5, constB=-0.5}
  NamedRerouteDeclaration(None=Multiply(A=NamedRerouteUsage(), B=Multiply(A=@n38, B=@n39))) {name=Center, nodeColor=<rgba 1 0 0.0937 1>}
  NamedRerouteUsage()
  NamedRerouteUsage()
  NamedRerouteUsage()
  ScalarParameter() {defaultValue=0.1, parameterName=Roughness}
  ScalarParameter() {defaultValue=0.5, parameterName=Metallic}
  ScalarParameter() {defaultValue=1, parameterName=Specular}
  n40 = Subtract(A=WorldPosition().XYZ, B=NamedRerouteDeclaration(None=ObjectPositionWS()) {name=Position, nodeColor=<rgba 0.4297 0 1 1>})
  n41 = Normalize(VectorInput=Multiply(A=NamedRerouteUsage(), B=Constant3Vector() {constant=<rgba -1 -1 1 0>}))
  n42 = Constant3Vector() {constant=<rgba 0 0 1 0>}
  n43 = ComponentMask(None=Divide(
    A=MaterialFunctionCall!(VectorToRotate=@n40, LookAtVector=@n41, RestingVector=@n42) {materialFunction=/Engine/Functions/Engine_MaterialFunctions01/Vectors/RotateVector}.Result
    B=NamedRerouteUsage()
   ) {constB=1}) {b=true}
  n44 = Add(A=Subtract(A=Multiply(B=@n43) {constA=-1}, B=@n39), B=@n36) {constB=0}
  SmoothStep(Value=Multiply(A=NamedRerouteUsage(), B=NamedRerouteDeclaration(None=@n44) {name=Water Line Gradient, nodeColor=<rgba 1 0 0.1172 1>}) {constA=1}) {constMax=0.001}
 ]
}
