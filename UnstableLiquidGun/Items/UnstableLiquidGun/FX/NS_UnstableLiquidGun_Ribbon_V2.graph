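# Niagara system that draws the Unstable Liquid Gun's liquid as ribbons with bubble sprites, driven by user position, ribbon ID, size, visibility and flashing arrays.
# TubeRibbon002 spawns points from the arrays, detects collisions, colors by flashing, PreExplodeIndex and time of day, and thins points by quality level.
# It writes start and end counters from PreExplodeIndex and lost visibility, which two GPU bubble emitters use to spawn bubbles along the stream.
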NiagaraSystem /UnstableLiquidGun/Items/UnstableLiquidGun/FX/NS_UnstableLiquidGun_Ribbon_V2
dependencies {
 usedDataInterfaces [&NiagaraDataInterfaceCurve, &NiagaraDataInterfacePlatformSet]
 usedDynamicInputs []
 usedModules [&SystemState, &self:ClearState]
}
emitters {
 Bubbles {
  simTarget=GPUComputeSim
  emitterUpdateScript [
   &EmitterState(Life Cycle Mode=::NewEnumerator1, Loop Behavior=::NewEnumerator1, Loop Duration Mode=::NewEnumerator1, UseLoopDelay=false)
   &self:CalcSpawnCount_01(
    LastKillCounter:NiagaraDataInterfaceSimpleCounter=$User.EndSpawnParticleCounter
    SpawnCounter:NiagaraDataInterfaceSimpleCounter=$User.StartSpawnParticleCounter
    SpawnRateScale:NiagaraFloat=&Multiply_Float(A=$User.BubbleSpawnRate, B=4)
   ) {module=CalcSpawnCount_01}
   &SpawnRate(SpawnRate=$Emitter.SpawnRateResult, Use Spawn Probability=false)
  ]
  particleSpawnScript [
   &self:SetParentIndex_01(
    LastKillCounter:NiagaraDataInterfaceSimpleCounter=$User.EndSpawnParticleCounter
    SpawnCounter:NiagaraDataInterfaceSimpleCounter=$User.StartSpawnParticleCounter
   ) {module=SetParentIndex_01}
   &InitializeParticle(
    Color Mode=::NewEnumerator0
    Lifetime Max=1.5
    Lifetime Min=0.25
    Lifetime Mode=::NewEnumerator1
    Sprite Size Mode=::NewEnumerator3
    Uniform Sprite Size=&RandomRangeFloat(Minimum=4, Maximum=5)
    UsePositionOffset=false
    Write Material Random=true
   )
   &self:InitialParticleWithArray_01(ArrayIndex:NiagaraInt32=$Particles.ParentIndex, PosArray:NiagaraDataInterfaceArrayFloat3=$User.PosArray) {module=InitialParticleWithArray_01}
   &SphereLocation(
    MaskSpawnGroup=false
    Offset=<0 0 3>
    Override Seed=false
    Sphere Radius=20
    Surface Only=false
    UseNonUniformScale=false
    UseOffset=true
    UseSphereOrientation=false
   )
   &AddVelocity(Velocity=&RandomRangeVector(Minimum=<-100 -100 150>, Maximum=<100 100 500>))
  ]
  particleUpdateScript [
   &ParticleState()
   &GravityForce(Gravity=<0 0 -2216.0457>)
   &Collision(
    Age Colliding Particles=false
    Friction=15
    Friction During a Bounce=15
    GPU Collision Type=::NewEnumerator2
    Module.Control Rest State Via Surface Normal=false
    Module.EnableMaxCollisionCount=false
    Module.Randomize Collision Normal Vector=false
    Percentage of Penetration Before Rest=0
    Restitution=0
   )
   &SolveForcesAndVelocity(Clamp Velocity=false, Limit Acceleration=false)
   &DynamicMaterialParameters(Module.Param0WriteEnabled=true, Module.Param1WriteEnabled=true, Module.Param2WriteEnabled=true, Module.Param3WriteEnabled=true)
   &ScaleSpriteSize(Uniform Curve Sprite Scale=DataInterface() {propertyValues {Curve {keys [
        {arriveTangent=5.5251, interpMode=RCIM_Cubic, leaveTangent=5.5251, tangentMode=RCTM_User, time=0, value=0}
        {interpMode=RCIM_Cubic, tangentMode=RCTM_User, time=0.5, value=1}
        {tangentMode=RCTM_User, time=1, value=0}
       ]}}})
   &ScaleSpriteSizeBySpeed(Max Scale Factor=<1 10>)
  ]
  rendererClasses [&NiagaraSpriteRendererProperties]
  renderers [{rendererClass=&NiagaraSpriteRendererProperties}]
 }
 Bubbles001 {
  simTarget=GPUComputeSim
  emitterUpdateScript [
   &EmitterState(Life Cycle Mode=::NewEnumerator1, Loop Behavior=::NewEnumerator1, Loop Duration Mode=::NewEnumerator1, UseLoopDelay=false)
   &self:CalcSpawnCount_02(
    LastKillCounter:NiagaraDataInterfaceSimpleCounter=$User.EndSpawnParticleCounter
    SpawnCounter:NiagaraDataInterfaceSimpleCounter=$User.StartSpawnParticleCounter
    SpawnRateScale:NiagaraFloat=&Multiply_Float(A=$User.BubbleSpawnRate, B=2)
   ) {module=CalcSpawnCount_02}
   &SpawnRate(SpawnRate=$Emitter.SpawnRateResult, Use Spawn Probability=false)
  ]
  particleSpawnScript [
   &self:SetParentIndex_02(
    LastKillCounter:NiagaraDataInterfaceSimpleCounter=$User.EndSpawnParticleCounter
    SpawnCounter:NiagaraDataInterfaceSimpleCounter=$User.StartSpawnParticleCounter
   ) {module=SetParentIndex_02}
   &InitializeParticle(
    Color Mode=::NewEnumerator0
    Lifetime Max=1
    Lifetime Min=0.4
    Lifetime Mode=::NewEnumerator1
    Sprite Size Mode=::NewEnumerator3
    Uniform Sprite Size=&RandomRangeFloat(Minimum=4, Maximum=7)
    UsePositionOffset=false
    Write Material Random=true
   )
   &self:InitialParticleWithArray_02(ArrayIndex:NiagaraInt32=$Particles.ParentIndex, PosArray:NiagaraDataInterfaceArrayFloat3=$User.PosArray) {module=InitialParticleWithArray_02}
   &SphereLocation(
    MaskSpawnGroup=false
    Offset=<0 0 3>
    Override Seed=false
    Sphere Radius=30
    Surface Only=false
    UseNonUniformScale=false
    UseOffset=true
    UseSphereOrientation=false
   )
   &AddVelocity(Velocity=&RandomRangeVector(Minimum=<-100 -100 50>, Maximum=<100 100 180>))
  ]
  particleUpdateScript [
   &ParticleState()
   &GravityForce(Gravity=&MakeVector(X=0, Y=0, Z=&RandomRangeFloat(Minimum=-600, Maximum=-1400)))
   &Collision(
    Age Colliding Particles=false
    Friction=15
    Friction During a Bounce=15
    GPU Collision Type=::NewEnumerator2
    Module.Control Rest State Via Surface Normal=false
    Module.EnableMaxCollisionCount=false
    Module.Randomize Collision Normal Vector=false
    Percentage of Penetration Before Rest=0
   ) {enabled=false}
   &CurlNoiseForce(Noise Strength=50, Pan Noise=false)
   &Drag(Use Linear Drag=true, Use Rotational Drag=false)
   &SolveForcesAndVelocity(Clamp Velocity=false, Limit Acceleration=false)
   &DynamicMaterialParameters(
    Index 0 Param 1=&Waveform(Global Phase=$Particles.Age, "Waveform [1]"=::NewEnumerator0, "[1] Phase"=$Particles.Age, "[1] Frequency"=3, "[1] Amplitude Min/Max"=<0 1>)
    Module.Param0WriteEnabled=true
    Module.Param1WriteEnabled=true
    Module.Param2WriteEnabled=true
    Module.Param3WriteEnabled=true
   )
   &ScaleSpriteSize(
    Uniform Curve Scale=&Lerp_Float(
     A=0.9
     B=1.3
     Alpha=&Waveform(Global Phase=$Particles.Age, "Waveform [1]"=::NewEnumerator0, "[1] Phase"=$Particles.Age, "[1] Frequency"=5, "[1] Amplitude Min/Max"=<0 1>)
    )
    Uniform Curve Sprite Scale=DataInterface() {propertyValues {Curve {keys [
        {arriveTangent=5.5251, interpMode=RCIM_Cubic, leaveTangent=5.5251, tangentMode=RCTM_User, time=0, value=0}
        <at 0.0613 1.2205>
        <at 0.2098 0.8295>
        <at 0.3613 0.6136>
        {interpMode=RCIM_Cubic, tangentMode=RCTM_User, time=0.5515, value=0.8977}
        <at 0.8332 0.1624>
        <at 0.9 0.7>
        {tangentMode=RCTM_User, time=1, value=0}
       ]}}}
   )
   &ScaleSpriteSizeBySpeed(Max Scale Factor=<1 10>) {enabled=false}
  ]
  rendererClasses [&NiagaraSpriteRendererProperties]
  renderers [{rendererClass=&NiagaraSpriteRendererProperties}]
 }
 TubeRibbon002 {
  emitterUpdateScript [
   &EmitterState(
    Inactive Response=::NewEnumerator1
    Life Cycle Mode=::NewEnumerator1
    Loop Behavior=::NewEnumerator1
    Loop Duration Mode=::NewEnumerator1
    UseLoopDelay=false
   )
   &self:SpawnFromArray(
    MaxSpawnCountPerFrame=100
    PosArray:NiagaraDataInterfaceArrayFloat3=$User.PosArray
    SpawnCounter:NiagaraDataInterfaceSimpleCounter=$User.SpawnCounter
   ) {module=SpawnFromArray}
   &SpawnPerFrame(Spawn Count:NiagaraInt32=$Emitter.SpawnFromArray.SpawnCount, Use Spawn Probability=false)
   /SetVariables(Emitter.LastReactionOffset=5, Emitter.ReactionSpeed=90, Emitter.StartReactionDelay=70) {enabled=false, isSetParameters=true}
   &FN_Timer(
    Decrement=true
    Decrement Multiplier=5
    Increment:NiagaraBool=&SetBoolByFloatComparison(
     A=&self:GetFloatFromArray(New Float Array:NiagaraDataInterfaceArrayFloat=$User.FirstParticleHittedAge)
     B=0
    )
    Increment Multiplier:NiagaraFloat=&MakeFloatFromInt(Int=$Emitter.ReactionSpeed)
   ) {enabled=false, module=FN_Timer002}
   &self:SpawnCounterIncrement(
    Amount:NiagaraInt32=/Niagara/DynamicInputs/Math/Min_Int(
     A:NiagaraInt32=/Niagara/DynamicInputs/Subtract/Subtract_Integer(
      A:NiagaraInt32=&MakeIntFromFloat(Float=&Subtract_Float(A=$StackContext.FN_Timer002.Value, B=&MakeFloatFromInt(Int=$Emitter.StartReactionDelay)))
      B:NiagaraInt32=&MakeIntFromFloat(Float=&self:GetSimpleCounterNumber(New Simple Counter:NiagaraDataInterfaceSimpleCounter=$User.StartSpawnParticleCounter))
     )
     B:NiagaraInt32=&Subtract_Integer(
      A:NiagaraInt32=&Subtract_Integer(A:NiagaraInt32=$Engine.Emitter.TotalSpawnedParticles, B:NiagaraInt32=$Emitter.LastReactionOffset)
      B:NiagaraInt32=&MakeIntFromFloat(Float=&self:GetSimpleCounterNumber(New Simple Counter:NiagaraDataInterfaceSimpleCounter=$User.StartSpawnParticleCounter))
     )
    )
    Execute:NiagaraBool=&Boolean_AndOperation(
     Bool1=&SetBoolByFloatComparison(
      A=&Subtract_Float(A=$StackContext.FN_Timer002.Value, B=&MakeFloatFromInt(Int=$Emitter.StartReactionDelay))
      B=&self:GetSimpleCounterNumber(New Simple Counter:NiagaraDataInterfaceSimpleCounter=$User.StartSpawnParticleCounter)
     )
     Bool2=&SetBoolByIntComparison(
      A=&MakeIntFromFloat(Float=&self:GetSimpleCounterNumber(New Simple Counter:NiagaraDataInterfaceSimpleCounter=$User.StartSpawnParticleCounter))
      B=&Subtract_Integer(A:NiagaraInt32=$Engine.Emitter.TotalSpawnedParticles, B:NiagaraInt32=$Emitter.LastReactionOffset)
      Comparison Type=::NewEnumerator6
     )
    )
    SpawnCounter:NiagaraDataInterfaceSimpleCounter=$User.StartSpawnParticleCounter
   ) {enabled=false, module=SpawnCounterIncrement002}
   &self:SpawnCounterIncrement(
    Amount:NiagaraInt32=&Subtract_Integer(
     A:NiagaraInt32=&MakeIntFromFloat(Float=&Subtract_Float(A=$StackContext.FN_Timer002.Value, B=&MakeFloatFromInt(Int=$Emitter.StartReactionDelay)))
     B:NiagaraInt32=&MakeIntFromFloat(Float=&self:GetSimpleCounterNumber(New Simple Counter:NiagaraDataInterfaceSimpleCounter=$User.RibbonColorCount))
    )
    Execute:NiagaraBool=&Boolean_AndOperation(
     Bool1=&SetBoolByFloatComparison(
      A=&Subtract_Float(A=$StackContext.FN_Timer002.Value, B=&MakeFloatFromInt(Int=$Emitter.StartReactionDelay))
      B=&self:GetSimpleCounterNumber(New Simple Counter:NiagaraDataInterfaceSimpleCounter=$User.RibbonColorCount)
     )
     Bool2=true
    )
    SpawnCounter:NiagaraDataInterfaceSimpleCounter=$User.RibbonColorCount
   ) {enabled=false, module=SpawnCounterIncrement004}
   /SetVariables(Emitter.ClampedPreExplodeIndex:NiagaraInt32=&MakeIntFromFloat(Float=&ClampFloat(
      Float=&MakeFloatFromInt(Int=$User.PreExplodeIndex)
      Min=-10
      Max=&Subtract_Float(A=&MakeFloatFromInt(Int=$Engine.Emitter.TotalSpawnedParticles), B=5)
     ))) {isSetParameters=true}
   &self:SetSimpleCounter(New Simple Counter:NiagaraDataInterfaceSimpleCounter=$User.StartSpawnParticleCounter, New int32:NiagaraInt32=$Emitter.ClampedPreExplodeIndex) {module=SetSimpleCounter}
  ]
  particleSpawnScript [
   &InitializeParticle(Lifetime=20, Sprite Size Mode=::NewEnumerator3, Uniform Sprite Size=40, UsePositionOffset=false, Write Material Random=true)
   &InitializeRibbon(Lifetime=10, Ribbon Width=50, UsePositionOffset=false, Write Material Random=true)
   &self:SetArrayIDAndLinkOrder() {module=SetArrayIDAndLinkOrder}
   &self:SpawnCounterIncrement(Amount=1, Execute=true, SpawnCounter:NiagaraDataInterfaceSimpleCounter=$User.SpawnCounter) {module=SpawnCounterIncrement}
   /SetVariables() {isSetParameters=true}
   /SetVariables() {isSetParameters=true}
   /SetVariables(Particles.RibbonU0Override=&Multiply_Float(
     A=&Add_Float(
      A=&MakeFloatFromInt(Int=&ReturnParticleIDAttributes())
      B=$Emitter.Age
     )
     B=0.1
    )) {isSetParameters=true}
  ]
  particleUpdateScript [
   &ParticleState()
   &Collision(
    Age Colliding Particles=false
    CPU Collision Trace Channel=ECC_GameTraceChannel5
    Collision Enabled=&SetBoolByFloatComparison(A=$Particles.Age, B=0.1)
    Friction=25
    Friction During a Bounce=25
    Module.Control Rest State Via Surface Normal=false
    Module.EnableMaxCollisionCount=false
    Module.Randomize Collision Normal Vector=false
    Particle Radius Scale=1.5
    Percentage of Penetration Before Rest=0
    Rest State Time Range=0
   )
   &SolveForcesAndVelocity(Clamp Velocity=false, Limit Acceleration=false)
   /SetVariables(Particles.Hitted:NiagaraBool=&Boolean_OrOperation(Bool1=$Particles.HasCollided, Bool2=$Particles.Hitted)) {isSetParameters=true}
   &self:SetRibbonAttributesFromArray(
    "Hitted？":NiagaraBool=$Particles.Hitted
    IsFlashingArray:NiagaraDataInterfaceArrayBool=$User.bIsFlashingArray
    IsFlashingFloatArray:NiagaraDataInterfaceArrayFloat=$User.fIsFlashingArray
    MyID:NiagaraInt32=$Particles.MyID
    PosArray:NiagaraDataInterfaceArrayFloat3=$User.PosArray
    RibbonIDArray:NiagaraDataInterfaceArrayInt32=$User.RibbonIDArray
    RibbonSizeArray:NiagaraDataInterfaceArrayInt32=$User.RibbonSizes
    ScaleRibbon:NiagaraFloat=&FloatFromCurve(
     FloatCurve=DataInterface() {propertyValues {Curve {keys [<at 0 0 RCIM_Cubic>, {arriveTangent=1, interpMode=RCIM_Cubic, leaveTangent=1, time=0.15, value=0.75}, <at 1 1>]}}}
     CurveIndex=$Particles.Age
     Scale Curve=0.06
    )
    VisibilityArray:NiagaraDataInterfaceArrayBool=$User.RibbonVisibilityArray
   ) {module=SetRibbonAttributesFromArray}
   &KillParticles(Kill Particles=$Particles.Kill) {enabled=false}
   &FN_Timer(
    Clamp=true
    Increment:NiagaraBool=&Boolean_OrOperation(
     Bool1=$Particles.Hitted
     Bool2=&SetBoolByIntComparison(A=$Particles.MaxCollisionIndex, B=$Particles.MyID, Comparison Type=::NewEnumerator0)
    )
    Increment Multiplier=2
   )
   &DoOnce(Trigger Condition:NiagaraBool=&Boolean_AndOperation(Bool1=$Particles.Hitted, Bool2=&SetBoolByIntComparison(A=$Particles.MyID, B=0, Comparison Type=::NewEnumerator4))) {module=DoOnce003}
   &self:SetFirstParticleHittedAge(
    A:NiagaraDataInterfaceArrayFloat=$User.FirstParticleHittedAge
    Age:NiagaraFloat=$Particles.Age
    SkipSet:NiagaraBool=&Boolean_NotOperation(Bool=$Particles.DoOnce003.Execute)
   ) {module=SetFirstParticleHittedAge}
   &DoOnce(Trigger Condition:NiagaraBool=&SetBoolByFloatComparison(
     A=$Particles.FN_Timer002.Value
     B=&self:GetSimpleCounterNumber(New Simple Counter:NiagaraDataInterfaceSimpleCounter=$User.StartSpawnParticleCounter)
    )) {enabled=false, module=DoOnce004}
   &FN_Timer(
    Increment:NiagaraBool=&Boolean_OrOperation(
     Bool1=$Particles.Hitted
     Bool2=&SetBoolByIntComparison(A=$Particles.MaxCollisionIndex, B=$Particles.MyID, Comparison Type=::NewEnumerator0)
    )
    Increment Multiplier=0.5
   ) {enabled=false, module=FN_Timer001}
   &DoOnce(Trigger Condition:NiagaraBool=&SetBoolByFloatComparison(A=$Particles.FN_Timer001.Value, B=0.3)) {enabled=false, module=DoOnce001}
   &self:SpawnCounterIncrement(Execute:NiagaraBool=$Particles.DoOnce001.Execute, SpawnCounter:NiagaraDataInterfaceSimpleCounter=$User.StartSpawnParticleCounter) {enabled=false, module=SpawnCounterIncrement003}
   &DoOnce(Trigger Condition:NiagaraBool=&Boolean_NotOperation(Bool=$Particles.Visibility)) {module=DoOnce002}
   &self:SpawnCounterIncrement(Amount=1, Execute:NiagaraBool=$Particles.DoOnce002.Execute, SpawnCounter:NiagaraDataInterfaceSimpleCounter=$User.EndSpawnParticleCounter) {module=SpawnCounterIncrement001}
   &DoOnce(Trigger Condition:NiagaraBool=$Particles.Hitted)
   &self:GenerateCustomEvent(Execute:NiagaraBool=$Particles.DoOnce.Execute, Int:NiagaraInt32=$Particles.MyID) {module=GenerateCustomEvent}
   /SetVariables() {isSetParameters=true}
   &ScaleColor(
    Scale RGB=&Add_Vector(
     A=&MakeVector(
      X=&Lerp_Float(
       A=1
       B=&Waveform(
        "Waveform [1]"=::NewEnumerator0
        "[1] Frequency"=3
        "[1] Amplitude Min/Max"=<0 1>
        "[1] Scale"=&Multiply_Float(A=&FNTimeOfDayScalar(Night Scale=0.1, Morning Scale=0.2, Evening Scale=0.3), B=30)
        "[1] Bias"=0.5
       )
       Alpha=$Particles.fIsFlashing
      )
      Y=&Lerp_Float(
       A=1
       B=&FNTimeOfDayScalar(Night Scale=3, Morning Scale=0.5, Evening Scale=3)
       Alpha=$Particles.fIsFlashing
      )
      Z=&Lerp_Float(
       A=1
       B=&FNTimeOfDayScalar(Night Scale=2, Morning Scale=0.3, Evening Scale=2)
       Alpha=$Particles.fIsFlashing
      )
     )
     B=&VectorFromFloat(Value=&FloatFromCurve(
       FloatCurve=DataInterface() {propertyValues {Curve {keys [{tangentMode=RCTM_User, time=0, value=0}, {tangentMode=RCTM_User, time=1, value=1}]}}}
       CurveIndex=&ClampFloat(
        Float=&Divide_Float(
         A=&Subtract_Float(
          A=&Add_Float(A=&MakeFloatFromInt(Int=$User.PreExplodeIndex), B=-210)
          B=&MakeFloatFromInt(Int=&Add_Integer(A=$Particles.MyID, B=0))
         )
         B=6
        )
        Min=-500
        Max=500
       )
       Scale Curve=&Multiply_Float(A=&FNTimeOfDayScalar(Night Scale=0.3, Morning Scale=1, Evening Scale=0.3), B=5)
      ))
    )
    ScaleA=true
    ScaleRGB=true
   ) {module=ScaleColor003}
   &DynamicMaterialParameters(
    Index 0 Param 1=&FloatFromCurve(
     FloatCurve=DataInterface() {propertyValues {Curve {keys [{interpMode=RCIM_Cubic, leaveTangent=4.5991, tangentMode=RCTM_Break, time=0, value=0}, <at 0.8 1>]}}}
     CurveIndex=$Particles.Age
    )
    Index 0 Param 2=&Lerp_Float(A=0, B=&Multiply_Float(A=$Particles.RibbonWidth, B=0.3), Alpha=$Particles.FN_Timer.Value)
    Index 0 Param 3=&Lerp_Float(A=0, B=2.2, Alpha=&MakeCustomFloatFromBool(Bool=$Particles.Hitted, True Float=1, False Float=0))
    Index 1 Param 1=&Lerp_Float(A=0, B=1, Alpha=$Particles.fIsFlashing)
    Index 1 Param 2=&FloatFromCurve(
     FloatCurve=DataInterface() {propertyValues {Curve {keys [{tangentMode=RCTM_User, time=0, value=0}, <at 0.7 0>, {tangentMode=RCTM_User, time=0.8, value=1}]}}}
     CurveIndex=&Divide_Float(A=$Particles.RibbonWidth, B=100)
     Scale Curve=0.8
    )
    Index 3 Param 1=&MakeCustomFloatFromBool(
     Bool=&Boolean_OrOperation(
      Bool1=$Particles.Hitted
      Bool2=&SetBoolByIntComparison(A=$Particles.MaxCollisionIndex, B=$Particles.MyID, Comparison Type=::NewEnumerator0)
     )
     True Float=1
     False Float=0
    )
    Module.Param0WriteEnabled=true
    Module.Param0WriteEnabled1=true
    Module.Param0WriteEnabled3=true
    Module.Param1WriteEnabled=true
    Module.Param1WriteEnabled1=true
    Module.Param1WriteEnabled3=false
    Module.Param2WriteEnabled=true
    Module.Param2WriteEnabled1=true
    Module.Param2WriteEnabled3=false
    Module.Param3WriteEnabled=false
    Module.Param3WriteEnabled1=true
    Module.Param3WriteEnabled3=false
    Write Parameter Index 1=true
    Write Parameter Index 3=true
   ) {module=DynamicMaterialParameters001}
   &KillParticles(Kill Particles=&Boolean_AndOperation(
     Bool1=&SetBoolByIntComparison(
      A=&ModuloInt(
       A=$Particles.MyID
       B=&self:SelectIntWithPlatform(
        Pass:NiagaraInt32=&self:SelectIntWithPlatform(
         Pass:NiagaraInt32=1
         Platform:NiagaraDataInterfacePlatformSet=DataInterface() {propertyValues {Platforms {qualityLevelMask=-31}}}
         Target:NiagaraInt32=3
        )
        Platform:NiagaraDataInterfacePlatformSet=DataInterface() {propertyValues {Platforms {qualityLevelMask=-30}}}
        Target:NiagaraInt32=2
       )
      )
      B=0
      Comparison Type=::NewEnumerator0
     )
     Bool2=&Boolean_NotOperation(Bool=$Transient.FirstFrame)
    )) {module=KillParticles001}
  ]
  rendererClasses [&NiagaraRibbonRendererProperties, &NiagaraSpriteRendererProperties]
  renderers [
   {rendererClass=&NiagaraRibbonRendererProperties}
   {rendererClass=&NiagaraRibbonRendererProperties, rendererIndex=1}
   {rendererClass=&NiagaraRibbonRendererProperties, rendererIndex=2}
   {rendererClass=&NiagaraSpriteRendererProperties, rendererIndex=3}
  ]
 }
}
summary {
 systemName=NS_UnstableLiquidGun_Ribbon_V2
 emitters [
  {
   emitterName=TubeRibbon002
   rendererClasses [&NiagaraRibbonRendererProperties, &NiagaraSpriteRendererProperties]
  }
  {emitterName=Bubbles, simTarget=GPUComputeSim, rendererClasses [&NiagaraSpriteRendererProperties]}
  {emitterName=Bubbles001, simTarget=GPUComputeSim, rendererClasses [&NiagaraSpriteRendererProperties]}
 ]
 userVariables [
  User.bIsFlashing {type=NiagaraBool}
  User.bIsFlashingArray {
   type=NiagaraDataInterfaceArrayBool
   description=""
   defaultValue {
    struct=NiagaraExt_VariableValue_DataInterface
    value {
     dataInterface=&self:NiagaraDataInterfaceArrayBool_2
     dataInterfaceClass=/Script/Niagara.NiagaraDataInterfaceArrayBool
    }
   }
  }
  User.BubbleSpawnRate {type=NiagaraFloat, defaultValue=1, description=""}
  User.EndSpawnParticleCounter {
   type=NiagaraDataInterfaceSimpleCounter
   description=""
   defaultValue {
    struct=NiagaraExt_VariableValue_DataInterface
    value {
     dataInterface=&self:NiagaraDataInterfaceSimpleCounter_1
     dataInterfaceClass=/Script/Niagara.NiagaraDataInterfaceSimpleCounter
    }
   }
  }
  User.FirstParticleHittedAge {
   type=NiagaraDataInterfaceArrayFloat
   defaultValue {
    struct=NiagaraExt_VariableValue_DataInterface
    value {
     dataInterface=&self:NiagaraDataInterfaceArrayFloat_0
     dataInterfaceClass=&NiagaraDataInterfaceArrayFloat
    }
   }
  }
  User.fIsFlashingArray {
   type=NiagaraDataInterfaceArrayFloat
   defaultValue {
    struct=NiagaraExt_VariableValue_DataInterface
    value {
     dataInterface=&self:NiagaraDataInterfaceArrayFloat_1
     dataInterfaceClass=&NiagaraDataInterfaceArrayFloat
    }
   }
  }
  User.PosArray {
   type=NiagaraDataInterfaceArrayFloat3
   description=""
   defaultValue {
    struct=NiagaraExt_VariableValue_DataInterface
    value {
     dataInterface=&self:NiagaraDataInterfaceArrayFloat3_0
     dataInterfaceClass=&NiagaraDataInterfaceArrayFloat3
    }
   }
  }
  User.PreExplodeIndex {type=NiagaraInt32, defaultValue=-50}
  User.RibbonColorCount {
   type=NiagaraDataInterfaceSimpleCounter
   description=""
   defaultValue {
    struct=NiagaraExt_VariableValue_DataInterface
    value {
     dataInterface=&self:NiagaraDataInterfaceSimpleCounter_2
     dataInterfaceClass=&NiagaraDataInterfaceSimpleCounter
    }
   }
  }
  User.RibbonHittedArray {
   type=NiagaraDataInterfaceArrayBool
   description=""
   defaultValue {
    struct=NiagaraExt_VariableValue_DataInterface
    value {
     dataInterface=&self:NiagaraDataInterfaceArrayBool_0
     dataInterfaceClass=&NiagaraDataInterfaceArrayBool
    }
   }
  }
  User.RibbonIDArray {
   type=NiagaraDataInterfaceArrayInt32
   description=""
   defaultValue {
    struct=NiagaraExt_VariableValue_DataInterface
    value {
     dataInterface=&self:NiagaraDataInterfaceArrayInt32_0
     dataInterfaceClass=&NiagaraDataInterfaceArrayInt32
    }
   }
  }
  User.RibbonSizes {
   type=NiagaraDataInterfaceArrayInt32
   description=""
   defaultValue {
    struct=NiagaraExt_VariableValue_DataInterface
    value {
     dataInterface=&self:NiagaraDataInterfaceArrayInt32_1
     dataInterfaceClass=&NiagaraDataInterfaceArrayInt32
    }
   }
  }
  User.RibbonVisibilityArray {
   type=NiagaraDataInterfaceArrayBool
   description=""
   defaultValue {
    struct=NiagaraExt_VariableValue_DataInterface
    value {
     dataInterface=&self:NiagaraDataInterfaceArrayBool_1
     dataInterfaceClass=&NiagaraDataInterfaceArrayBool
    }
   }
  }
  User.SpawnCounter {
   type=NiagaraDataInterfaceSimpleCounter
   description=""
   defaultValue {
    struct=NiagaraExt_VariableValue_DataInterface
    value {
     dataInterface=&self:NiagaraDataInterfaceSimpleCounter_0
     dataInterfaceClass=&NiagaraDataInterfaceSimpleCounter
    }
   }
  }
  User.StartSpawnParticleCounter {
   type=NiagaraDataInterfaceSimpleCounter
   description=""
   defaultValue {
    struct=NiagaraExt_VariableValue_DataInterface
    value {
     dataInterface=&self:NiagaraDataInterfaceSimpleCounter_3
     dataInterfaceClass=&NiagaraDataInterfaceSimpleCounter
    }
   }
  }
 ]
}
